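# Default Lit material for Niagara mesh particles in the Sunlit skydiving contrail.
# Dynamic parameters drive time-based sine and cosine wiggle on three axes, plus an oscillating rotation about the particle pivot for vertex offset.
# Normals, optionally mapped, get a matching rotation correction, skipped on Low and Medium quality.
# Texture or parameter base color is multiplied by a curve-atlas Fresnel reflectance lookup, bypassed on Low; texture alpha can supply roughness.
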Material /BRCosmetics/Athena/Effects/Particles/Skydiving/NS_Contrail_SunLit/Materials/M_Contrail_Sunlit_Meshes
graph {
 properties {bCastRayTracedShadows=false, bUsedWithNiagaraMeshParticles=true, pixelDepthOffsetMode=PDOM_Legacy, usedShadingModels=Default Lit}
 nodes [
  Constant() {r=1}
  ObjectPositionWS()
  ParticlePositionWS()
  n1 = TextureSampleParameter2D() {
   group=Normal
   parameterName=NormalMap
   samplerType=SAMPLERTYPE_Normal
   texture=/BRCosmetics/Athena/Effects/Particles/Skydiving/NS_Contrail_SunLit/Textures/T_Contrail_Sunlit_Rings_N
  }
  n2 = StaticSwitchParameter(True=@n1.RGB, False=Constant3Vector() {constant=<rgba 0 0 1 0>}) {group=Normal, parameterName=UseNormalMap}
  n3 = DynamicParameter() {paramNames [WiggleOffset, WiggleRate, WiggleAmplitude, Rotation]}
  n4 = Add(A=Time(), B=@n3.WiggleOffset) {constB=0}
  n5 = Multiply(A=@n4, B=@n3.WiggleRate)
  n6 = Multiply(A=@n5) {constB=0.4}
  n7 = Sine(None=@n6)
  n8 = Cosine(None=@n6)
  n9 = Normalize(VectorInput=AppendVector(A=AppendVector(A=@n7, B=@n8), B=Constant()))
  n10 = Multiply(A=Multiply(A=Sine(None=@n6)) {constB=0.04}, B=DynamicParameter() {paramNames [WiggleOffset, WiggleRate, WiggleAmplitude, Rotation]}.Rotation)
  n11 = MaterialFunctionCall(Rotation Axis=@n9, Rotation Angle=@n10, World Space Vertex Normals=Transform(None=@n2)) {materialFunction=&FixRotateAboutAxisNormals}
  QualitySwitch(
   Default=@n11.New Tangent Space Vertex Normal
   Low=@n2
   High=@n11.New Tangent Space Vertex Normal
   Medium=@n2
   Epic=@n11.New Tangent Space Vertex Normal
  )
  n12 = Sine(None=@n5)
  n13 = Cosine(None=@n5) {period=1.2}
  n14 = Add(
   A=Multiply(A=Constant3Vector() {constant=<rgba 1 0 0 0>}, B=Multiply(A=@n12, B=@n3.WiggleAmplitude) {constB=0.9}) {constB=0}
   B=Multiply(A=Multiply(A=@n13, B=@n3.WiggleAmplitude) {constB=0.9}, B=Constant3Vector() {constant=<rgba 0 1 0 0>})
  )
  n15 = Multiply(A=@n4) {constB=1.1}
  n16 = Multiply(A=Multiply(A=Sine(None=@n15) {period=-0.8}, B=@n3.WiggleAmplitude), B=Constant3Vector() {constant=<rgba 0 0 1 0>})
  n17 = MaterialFunctionCall() {materialFunction=&ObjectPivotPoint}
  n18 = WorldPosition()
  Reroute(None=Add(
    A=Add(A=@n14, B=@n16)
    B=RotateAboutAxis(NormalizedRotationAxis=@n9, RotationAngle=@n10, PivotPoint=@n17.Mesh Particle Pivot Location, Position=@n18.XYZ)
   ))
  n19 = TextureSampleParameter2D() {
   group=Base_Color
   parameterName=Base_Color_Tex
   texture=/BRCosmetics/Athena/Effects/Particles/Skydiving/NS_Contrail_SunLit/Textures/T_Contrail_Sunlit_Rings_D
  }
  n20 = CameraVectorWS()
  n21 = PixelNormalWS()
  n22 = CurveAtlasRowParameter(CurveTime=Saturate(None=DotProduct(A=@n20, B=@n21))) {
   atlas=&AthenaCurveAtlas01
   curve=/BRCosmetics/Characters/Player/Female/Medium/Bodies/F_MED_Ms_Alpine/Textures/AluminiumreflectanceCurve
   defaultValue=189
   group=LUT
   parameterName=FresnelReflectance_LUT
  }
  n23 = QualitySwitch(Default=@n22.R, Low=Constant() {r=1}, High=@n22.R, Medium=@n22.R, Epic=@n22.R)
  StaticSwitchParameter(
   True=Multiply(A=@n19.RGB, B=@n23)
   False=Multiply(A=VectorParameter() {parameterName=Base Color, defaultValue=<rgba 1 0.7167 0.5 1>}.RGB, B=@n23)
  ) {group=Base_Color, parameterName=UseBaseColorTexture}
  StaticSwitchParameter(True=@n19.A, False=ScalarParameter() {defaultValue=0.5, parameterName=Roughness}) {group=Base_Color, parameterName=UseBaseColorTexture}
  TextureSample() {samplerType=SAMPLERTYPE_Normal, texture=/BRCosmetics/Weapons/FORT_Melee/Pickaxe_Sunlit/Textures/T_Pickaxe_Sunlit_N}
  VectorParameter() {parameterName=Base Color1, defaultValue=<rgba 1 0.6429 0.3698 1>}
 ]
}
