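# Unlit translucent Niagara ribbon material for the SunLit skydiving contrail.
# Panning smoke noise distorts a vertical gradient mask that shapes the trail.
# Emissive combines a scrolling pattern, rainbow-tinted Color A and a Color B caustic glow at one end, with time-of-day scaled colors.
# Opacity uses noise dissolve offset by particle alpha, near-camera fade and quality-switched depth fade, while dynamic parameters offset and scale the layers.
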Material /BRCosmetics/Athena/Effects/Particles/Skydiving/NS_Contrail_SunLit/Materials/M_SunLitRibbon1
graph {
 properties {
  bCastRayTracedShadows=false
  bUsedWithNiagaraRibbons=true
  blendMode=BLEND_Translucent
  pixelDepthOffsetMode=PDOM_Legacy
  shadingModel=MSM_Unlit
  usedShadingModels=Unlit
 }
 nodes [
  Constant() {r=1}
  n1 = ScalarParameter() {defaultValue=0.65, group=Pattern_Main, parameterName=Pattern_Main_Intensity}
  n2 = TextureCoordinate()
  n3 = Multiply(A=ComponentMask(None=@n2) {g=true}, B=ScalarParameter() {defaultValue=2, group=Pattern_Main, parameterName=Pattern_V_Scale, sortPriority=34}) {constB=10}
  n4 = Multiply(A=ComponentMask(None=@n2) {r=true}, B=ScalarParameter() {defaultValue=1, group=Pattern_Main, parameterName=Pattern_U_Scale, sortPriority=35}) {constB=20}
  n5 = DynamicParameter() {paramNames [Flame_Offset, FE_Tweak, Pattern_Offset, Pattern_Scale]}
  n6 = Multiply(A=Add(A=AppendVector(A=@n3, B=@n4), B=@n5.Pattern_Offset) {constB=0.5}, B=@n5.Pattern_Scale)
  n7 = AppendVector(
   A=Constant()
   B=Multiply(A=ScalarParameter() {defaultValue=0.55, group=Pattern_Main, parameterName=Pattern_PanSpeed_Main, sortPriority=33}) {constB=-1}
  )
  n8 = TextureSampleParameter2D(UVs=VertexInterpolator(VS=Panner(Coordinate=@n6, Speed=@n7) {speedY=-1.8}).PS) {
   parameterName=TexturePattern
   samplerType=SAMPLERTYPE_LinearGrayscale
   texture=/BRCosmetics/Athena/Effects/Particles/Skydiving/NS_Contrail_SunLit/Textures/Pattern_SunLit_Main
  }
  n9 = TextureCoordinate()
  n10 = TextureCoordinate()
  n11 = ComponentMask(None=@n10) {r=true}
  n12 = ComponentMask(None=@n10) {g=true}
  n13 = AppendVector(A=Multiply(A=@n11) {constB=0.35}, B=Multiply(A=@n12) {constB=0.6})
  n14 = DynamicParameter() {paramNames [Flame_Offset, FE_Tweak, Pattern_Offset, Pattern_Scale]}
  n15 = TextureSampleParameter2D(UVs=Add(A=Panner(Coordinate=@n13) {speedX=-0.8, speedY=0.25}, B=@n14.Flame_Offset)) {parameterName=Distort_texture, texture=&T_Inky_Smoke_Tile}
  n16 = Multiply(A=Multiply(A=Power(Base=@n15.R)) {constB=2}, B=Saturate(None=Power(Base=@n11) {constExponent=1.2})) {constB=0.8}
  n17 = OneMinus(None=@n12)
  n18 = Power(Base=Multiply(A=Multiply(A=@n17, B=@n12)) {constB=3}) {constExponent=3}
  n19 = Add(A=@n9, B=Multiply(A=Multiply(A=@n16, B=@n18), B=ScalarParameter() {defaultValue=6, parameterName=Distort_Power}) {constB=2})
  n20 = TextureSampleParameter2D(UVs=@n19) {parameterName=Texture, samplerType=SAMPLERTYPE_Masks, texture=/Game/Global/Textures/Default/Masks_and_Gradients/T_VertGradient_MASK}
  n21 = Power(Base=@n20.R) {constExponent=4}
  n22 = ComponentMask(None=@n9) {r=true}
  n23 = OneMinus(None=@n22)
  n24 = LinearInterpolate(A=Multiply(A=@n21) {constB=1.4}, B=Multiply(A=@n21) {constB=6}, Alpha=Saturate(None=Power(Base=@n23) {constExponent=1.4}))
  n25 = Multiply(A=@n23)
  n26 = Power(Base=@n23) {constExponent=3}
  n27 = Saturate(None=Multiply(A=Saturate(None=@n24), B=Add(A=@n25, B=@n26)))
  n28 = Power(Base=@n27)
  n29 = Multiply(A=Multiply(A=Multiply(A=@n1, B=@n8.RGB) {constB=1.7}, B=Saturate(None=@n28)), B=Constant() {r=1})
  n30 = Constant4Vector() {constant=<rgba 1 1 1 0.5>}
  n31 = Multiply(
   A=Constant() {r=1.5}
   B=MaterialFunctionCall(Value_Morning=@n30.R, Value_Day=@n30.R, Value_Evening=@n30.R, Value_Night=@n30.R) {materialFunction=&Scalar_Time_of_Day}.Result
  )
  n32 = Multiply(
   A=ParticleColor().RGB
   B=Multiply(A=VectorParameter() {group=Color, parameterName=Color A, defaultValue=<rgba 0.3167 0.076 1.52 1>}.RGB, B=@n31) {constB=2}
  )
  n33 = LinearInterpolate(A=Constant() {r=1}, B=@n32) {constAlpha=1}
  n34 = Multiply(A=TextureCoordinate()) {constB=0.3}
  n35 = Multiply(A=TextureSampleParameter2D(UVs=Panner(Coordinate=@n34) {speedX=-0.4, speedY=0.16}) {parameterName=Distort_texture, texture=&T_Inky_Smoke_Tile}.R) {constB=0.25}
  n36 = Add(A=DynamicParameter() {paramNames [Flame_Offset, FE_Tweak, Pattern_Offset, Pattern_Scale]}.Flame_Offset) {constB=0.1}
  n37 = TextureCoordinate()
  n38 = VectorParameter() {group=Rainbow, parameterName=Rainbow UV, sortPriority=1, defaultValue=<rgba 1 0.7 -0.6 -0.05>}
  n39 = Panner(Coordinate=Multiply(A=Add(A=@n36, B=@n37), B=AppendVector(A=@n38.R, B=@n38.R)), Speed=AppendVector(A=@n38.B, B=@n38.B))
  n40 = Desaturation(
   None=TextureSample(UVs=Add(A=@n35, B=@n39)) {texture=&T_Tile_RainbowColor}.RGB
   Fraction=ScalarParameter() {defaultValue=0.3, group=Rainbow, parameterName=Rainbow Desaturation}
  )
  n41 = ScalarParameter() {defaultValue=4.8, group=Rainbow, parameterName=Rainbow Intensity}
  n42 = Multiply(
   A=LinearInterpolate(B=Multiply(A=@n40, B=@n41), Alpha=ScalarParameter() {defaultValue=0.52, group=Rainbow, parameterName=Use Rainbow}) {constA=1}
   B=@n32
  )
  n43 = VectorParameter() {group=Color, parameterName=Color B, defaultValue=<rgba 0 0.25 1 1>}
  n44 = OneMinus(None=@n11)
  n45 = DynamicParameter() {paramNames [Flame_Offset, FE_Tweak, Pattern_Offset, Pattern_Scale]}
  n46 = TextureCoordinate()
  n47 = Panner(Coordinate=Multiply(A=@n46, B=Constant2Vector() {g=0.7, r=1})) {speedX=-1}
  n48 = TextureSampleParameter2D(UVs=VertexInterpolator(VS=Add(A=@n45.Flame_Offset, B=@n47)).PS) {parameterName=Distort_texture2, texture=&T_ShieldCaustics}
  n49 = Add(A=Multiply(A=@n46, B=Constant2Vector() {g=0.5, r=0.8}) {constB=3}, B=@n14.Pattern_Offset)
  n50 = Constant() {r=0.25}
  n51 = TextureSampleParameter2D(UVs=Panner(Coordinate=MaterialFunctionCall(UVs=@n49, "Rotation Angle (0-1)"=@n50) {materialFunction=&CustomRotator}.Rotated Values) {speedX=-0.05, speedY=-2}) {parameterName=Distort_texture3, texture=/Game/Gadgets/Assets/VinderTech_GliderChute/Glider_Sunlit/FX/T_Glider_Noise_01}
  n52 = Power(Base=@n44) {constExponent=12}
  n53 = Saturate(None=LinearInterpolate(A=Multiply(A=@n51.RGB) {constB=2}, B=Multiply(A=@n52) {constB=5}) {constAlpha=0.4})
  n54 = Multiply(A=Multiply(A=Power(Base=@n44) {constExponent=8}, B=LinearInterpolate(A=@n48.R, Alpha=@n53) {constB=0.7})) {constB=10}
  n55 = Power(Base=@n44) {constExponent=4}
  n56 = Multiply(A=Power(Base=Multiply(A=@n48.R, B=@n55)) {constExponent=3}) {constB=50}
  n57 = Multiply(
   A=DynamicParameter() {paramNames [Flame_Offset, FE_Tweak, Pattern_Offset, Pattern_Scale]}.FE_Tweak
   B=Multiply(A=Multiply(A=@n43.RGB, B=@n31) {constB=3}, B=Add(A=@n54, B=@n56))
  )
  MaterialFunctionCall(Emissive=Add(A=Multiply(A=@n29, B=@n33), B=Add(A=@n42, B=@n57))) {materialFunction=&MF_FX_EmissiveAdjustments}
  n58 = Multiply(A=@n22) {constB=6}
  n59 = DepthFade(Opacity=@n27, FadeDistance=ScalarParameter() {defaultValue=10, parameterName=Fade_Distance})
  n60 = QualitySwitch(Default=@n59, Low=@n27, High=@n59, Medium=@n27, Epic=@n59)
  n61 = Add(
   A=Multiply(A=TextureCoordinate(), B=Constant2Vector() {g=0.4, r=0.2})
   B=Add(A=DynamicParameter() {paramNames [Flame_Offset, FE_Tweak, Pattern_Offset, Pattern_Scale]}.Flame_Offset) {constB=0.15}
  ) {constB=0.15}
  n62 = ParticleColor()
  n63 = Subtract(
   A=TextureSampleParameter2D(UVs=Panner(Coordinate=@n61) {speedX=-0.2, speedY=0.06}) {parameterName=Dissolve_texture, texture=/Game/Global/Textures/Default/Noise/T_FireNoise_01}.R
   B=@n62.A
  )
  n64 = Multiply(A=@n20.R, B=@n63)
  Multiply(
   A=Multiply(A=Saturate(None=@n58), B=Multiply(A=@n60, B=@n64))
   B=MaterialFunctionCall(Distance=Constant() {r=70}) {materialFunction=&Near_Camera_Fade}.Result
  )
  n65 = Saturate(None=OneMinus(None=@n62.A))
  QualitySwitch(Default=@n63, Low=@n65, High=@n63, Medium=@n65, Epic=@n63)
  VectorParameter() {parameterName=Color1, defaultValue=<rgba 0.6537 0 1 0>}
  VectorParameter() {parameterName=Color2, defaultValue=<rgba 0 0.6698 1 0>}
  VertexInterpolator(VS=@n19)
 ]
}
