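# Niagara skydiving contrail cosmetic (SunLit V3) that attaches effects to bones or sockets on the player's skeletal mesh.
# Mesh bands, orbiting foot pieces, spinning metallic pieces, flare and light sprites pick attach points by spawn group.
# System dive attributes, integrators and paraglide lean drive their offsets, orientation and spin.
# Time-of-day scalars adjust brightness and front-end modifiers substitute menu motion; the debug emitter and visible ribbon emitters are disabled.
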NiagaraSystem /BRCosmetics/Athena/Effects/Particles/Skydiving/NS_Contrail_SunLit/NS_Contrail_SunLit_V3
dependencies {
 usedDataInterfaces [&NiagaraDataInterfaceCurve, &NiagaraDataInterfaceColorCurve]
 usedDynamicInputs [
  &RotateVectorByQuaternion
  /Niagara/DynamicInputs/Rotation/InvertQuaternion
  &TransformVector
  &SetBoolByFloatComparison
  &NormalizeVector
 ]
 usedModules [
  &SystemState
  /Game/Effects/Niagara/Modules/Skydiving/FN_DivingAttributes/v5/FN_DivingAttributes
  /Game/Effects/Niagara/Modules/Skydiving/FN_DivingAttributes/v6/FN_DivingAttributes
  &self:SystemSpawnScript.SystemScriptSource.SystemScriptGraph.NiagaraNodeAssignment_2.SetVariables_#1
  &self:SystemSpawnScript.SystemScriptSource.SystemScriptGraph.NiagaraNodeAssignment_1.SetVariables_#2
  &FN_Integrator
  &self:SystemSpawnScript.SystemScriptSource.SystemScriptGraph.NiagaraNodeAssignment_0.SetVariables_#3
  &self:SystemSpawnScript.SystemScriptSource.SystemScriptGraph.NiagaraNodeAssignment_7.SetVariables_#4
 ]
}
emitters {
 Bands {
  emitterUpdateScript [
   &EmitterState(
    Inactive Response=::NewEnumerator1
    Life Cycle Mode=::NewEnumerator1
    Loop Behavior=::NewEnumerator1
    Loop Duration Mode=::NewEnumerator1
    UseLoopDelay=false
   )
   &SpawnBurst_Instantaneous(Spawn Count=4, Use Spawn Probability=false)
  ]
  particleSpawnScript [
   &FN_SetSpawnGroup()
   &InitializeParticle(
    Color Mode=::NewEnumerator0
    Lifetime=0
    Mesh Scale Mode=::NewEnumerator3
    Mesh Uniform Scale=&MaskFloatBySpawnGroup(
     Masked Float=0.08
     Passthrough Float=0.11
     Spawn Group Mask=&ModuloInt(A=&SpawnGroup(), B=2)
    )
    Sprite Rotation Mode=::NewEnumerator2
    UsePositionOffset=false
    Write Material Random=true
   ) {module=InitializeParticle001}
   &DynamicMaterialParameters(
    Index 0 Param 1=&RandomRangeFloat(Minimum=0, Maximum=20)
    Index 0 Param 2=&RandomRangeFloat(Minimum=0.9, Maximum=1.1)
    Index 0 Param 3=&RandomRangeFloat(Minimum=0.8, Maximum=1)
    Module.Param0WriteEnabled=true
    Module.Param1WriteEnabled=true
    Module.Param2WriteEnabled=true
    Module.Param3WriteEnabled=true
   )
   &Color(
    Color=<rgba 0.2625 0.341 1 1>
    Scale Color=&VectorFromFloat(Value=&MaskFloatBySpawnGroup(Masked Float=9, Passthrough Float=3))
   )
  ]
  particleUpdateScript [
   &ParticleState(Kill Particles When Lifetime Has Elapsed=false, Let Infinitely Lived Particles Die When Emitter Deactivates=true)
   &SkeletalMeshLocation(
    Bone / Socket Index=$Particles.SpawnGroup
    Bone Sampling Mode=::NewEnumerator4
    Inherit Velocity=false
    MaskSpawnGroup=false
    Sampled Position Offset=&MaskVectorBySpawnGroup(
     Masked Vector=<0 0 0>
     Passthrough Vector=<0 5 -5>
     Spawn Group Mask=&ModuloInt(A=&SpawnGroup(), B=2)
    )
    Sampled Position Offset Coordinate Space:ENiagaraMeshTransforms=::NewEnumerator2
    Skeletal Mesh=$User.SkeletalMesh
   )
   /SetVariables(Particles.MeshOrientation=&LerpQuaternion(
     Quaternion A:Quat4f=$Output.SkeletalMeshLocation.SampledRotation
     Quaternion B:Quat4f=&MultiplyQuaternion(Quaternion A=<0 0 1 1>, Quaternion B=$Output.SkeletalMeshLocation.SampledRotation)
     Lerp Factor:NiagaraFloat=$User.ParaGlideLeanAlpha
     Enable Spherical Linear Interpolation:NiagaraBool=false
    )) {enabled=false, isSetParameters=true}
   /SetVariables(Particles.MeshOrientation=&MultiplyQuaternion(Quaternion A=$Particles.MeshOrientation, Quaternion B=<1 1 0 1>)) {enabled=false, isSetParameters=true}
   /SetVariables(Particles.MeshOrientation=&MultiplyQuaternion(
     Quaternion A=$Output.SkeletalMeshLocation.SampledRotation
     Quaternion B=&FindQuatBetweenTwoVectors(
      From Vector=<0 0 1>
      To Vector=&Lerp_Vector(
       A=&MaskVectorBySpawnGroup(
        Masked Vector=<-1 1 0>
        Passthrough Vector=<3 -1 0>
        Spawn Group Mask=&ModuloInt(A=&SpawnGroup(), B=2)
       )
       B=&MaskVectorBySpawnGroup(
        Masked Vector=<-5 1 0>
        Passthrough Vector=<1 -1 0>
        Spawn Group Mask=&ModuloInt(A=&SpawnGroup(), B=2)
       )
       Alpha=$System.Dive
      )
     )
    )) {isSetParameters=true}
   /SetVariables(Particles.MeshOrientation=&MultiplyQuaternion(
     Quaternion A=$Particles.MeshOrientation
     Quaternion B=&MakeQuatFromAxisAngle(Axis=<0 0 1>, AngleInDegrees=&Multiply_Float(A=$System.FN_Integrator.Value, B=-50))
    )) {isSetParameters=true}
   &ScaleColor(
    Scale Alpha=&FNTimeOfDayScalar(Night Scale=0.75, Morning Scale=0.85, Evening Scale=0.85)
    Scale RGB=&VectorFromFloat(Value=&Lerp_Float(A=0.8, B=1, Alpha=$User.ParaGlideLeanAlpha))
    ScaleA=true
    ScaleRGB=true
   ) {module=ScaleColor001}
  ]
  rendererClasses [&NiagaraMeshRendererProperties, &NiagaraLightRendererProperties]
  renderers [
   {rendererClass=&NiagaraMeshRendererProperties, rendererIndex=0}
   {rendererClass=&NiagaraLightRendererProperties, rendererIndex=1}
  ]
 }
 FN_Debug001 {
  bEnabled=false
  emitterUpdateScript [
   &EmitterState()
   &SpawnBurst_Instantaneous(Spawn Count=1, Use Spawn Probability=false)
   &SpawnRate(SpawnRate=3, Use Spawn Probability=false)
  ]
  particleSpawnScript [/Niagara/Modules/Spawn/Initialization/InitializeParticle(
    Lifetime=5
    Module.Write Color=true
    Module.Write Lifetime=true
    Module.Write Mass=true
    Module.Write Material Random=true
    Module.Write Position=false
    Module.Write Scale=false
    Module.Write SpriteRotation=false
    Module.Write SpriteSize=true
    Position=&SimulationPosition()
    Sprite Size=<100 100>
   )]
  particleUpdateScript [
   &ParticleState()
   &FN_Timer()
   /SetVariables(Particles.Debug:NiagaraFloat=&FNTimeOfDayScalar(Night Scale=0.4, Morning Scale=0.5, Evening Scale=0.4)) {isSetParameters=true}
   &DynamicMaterialParameters(
    Index 0 Param 1=$User.ParaGlideLeanAlpha
    Module.Param0WriteEnabled=true
    Module.Param1WriteEnabled=true
    Module.Param2WriteEnabled=true
    Module.Param3WriteEnabled=true
   )
   &CameraOffset(Camera Offset Amount=111)
  ]
  rendererClasses [&NiagaraSpriteRendererProperties]
  renderers [{rendererClass=&NiagaraSpriteRendererProperties}]
 }
 Feet_Pieces {
  emitterSpawnScript [/SetVariables() {isSetParameters=true}]
  emitterUpdateScript [
   &EmitterState(Life Cycle Mode=::NewEnumerator1, Loop Behavior=::NewEnumerator1, Loop Duration Mode=::NewEnumerator1, UseLoopDelay=false)
   &SpawnBurst_Instantaneous(Spawn Count=10, Use Spawn Probability=false, UseLoopCountLimit=false)
   &SpawnRate(Spawn Probability=0.85, SpawnRate=&Lerp_Float(A=8, B=30, Alpha=$System.Dive), Use Spawn Probability=true) {enabled=false}
   /SetVariables(Emitter.FeetPiecesRotation:NiagaraFloat=&Add_Float(
     A=$Emitter.FeetPiecesRotation
     B=&Multiply_Float(A=$Engine.DeltaTime, B=&Lerp_Float(A=0.25, B=0.45, Alpha=$System.Dive))
    )) {isSetParameters=true}
  ]
  particleSpawnScript [
   /Game/Effects/Niagara/Modules/Skydiving/FN_DivingAttributes/FN_DivingAttributes_v2(Accumulate Dive=0.1, Accumulate Rest=-0.1, SkeletalMesh:NiagaraDataInterfaceSkeletalMesh=$User.SkeletalMesh) {enabled=false}
   &FN_SetSpawnGroup(Number of Spawn Groups=2)
   &InitializeParticle(
    Lifetime=&Lerp_Float(
     A=&RandomRangeFloat(Minimum=0.35, Maximum=0.6)
     B=&RandomRangeFloat(Minimum=0.08, Maximum=0.15)
     Alpha=$System.Dive
    )
    Mesh Scale=&VectorFromFloat(Value=&RandomRangeFloat(Minimum=0.08, Maximum=0.13))
    Module.Write Color=true
    Module.Write Lifetime=true
    Module.Write Mass=false
    Module.Write Material Random=true
    Module.Write Position=false
    Module.Write Scale=true
    Module.Write SpriteRotation=false
    Module.Write SpriteSize=false
    Sprite Rotation=&RandomRangeFloat(Minimum=0, Maximum=360)
    Sprite Size=&Vector2DFromFloat(Value=&RandomRangeFloat(Minimum=20, Maximum=30))
   )
   &InitialMeshOrientation(Mesh Orientation Mode=::NewEnumerator2, Rotation=<0 0 0.5>, Use Rotation Vector=true)
   &DynamicMaterialParameters(
    Index 0 Param 1=&RandomRangeFloat(Minimum=0, Maximum=20)
    Index 0 Param 2=&RandomRangeFloat(Minimum=1, Maximum=1.4)
    Index 0 Param 3=&RandomRangeFloat(Minimum=0.4, Maximum=0.8)
    Module.Param0WriteEnabled=true
    Module.Param1WriteEnabled=true
    Module.Param2WriteEnabled=true
    Module.Param3WriteEnabled=true
   )
   /SetVariables(Particles.VisTag:NiagaraInt32=&RandomRangeInt(Minimum=0, Maximum=1)) {isSetParameters=true}
  ]
  particleUpdateScript [
   &ParticleState(Kill Particles When Lifetime Has Elapsed=false, Let Infinitely Lived Particles Die When Emitter Deactivates=true)
   &SkeletalMeshLocation(
    Apply To Particle Position=false
    Bone / Socket Index=&Add_Integer(A=$Particles.SpawnGroup, B=2)
    Bone Sampling Mode=::NewEnumerator4
    Inherit Velocity=false
    "Inherit Velocity (Scale)"=0
    Interpolate Mesh Sampling=false
    MaskSpawnGroup=false
    Skeletal Mesh=$User.SkeletalMesh
   )
   &ShapeLocation(
    Cone Angle:NiagaraFloat=&Lerp_Float(A=38, B=43, Alpha=$System.FN_Integrator001.Value)
    Cone Inner Angle:NiagaraFloat=&Lerp_Float(A=38, B=43, Alpha=$System.FN_Integrator001.Value)
    Cone Length=40
    Cone Surface Distribution=0.05
    Flatten Endcaps=true
    MaskSpawnGroup=false
    Offset=&Multiply_Vector(A=$System.PiecesVelocity, B=<0 0 15>)
    Offset Mode=::NewEnumerator0
    Override Seed=false
    Rotation Angle:NiagaraFloat=&Multiply_Float(A=$Emitter.FeetPiecesRotation, B=&RandomRangeFloat(Minimum=-1, Maximum=1))
    Rotation Axis=<0 0 1>
    Rotation Mode=::NewEnumerator1
    Shape Origin:Vector3f=$Output.SkeletalMeshLocation.SampledPosition
    Shape Primitive=::NewEnumerator5
    Transform Order:ENiagara_TransformOrder=::NewEnumerator1
    UseOwnerScale=true
   )
   &UpdateMeshOrientation(
    Facing Direction=&Add_Vector(
     A=&Multiply_Vector(A=$System.PiecesVelocity, B=<-10 -10 0>)
     B=&Multiply_Vector(
      A=&SubtractVectorFromPosition(Position=$Particles.Position, Vector=$Output.SkeletalMeshLocation.SampledPosition)
      B=&Lerp_Vector(
       A=<1 1 0>
       B=<1 1 -0.1>
       Alpha=&Multiply_Float(
        A=&DistanceBetweenPositions(A=Linked() {linkedVariable {name=Output.SkeletalMeshLocation.SampledPosition, type=Vector3f}}, B=$Particles.Position)
        B=0.1
       )
      )
     )
    )
    Facing Interpolation=false
    Orientation Method=::NewEnumerator1
    Override Quat=false
    TransformFacing=false
    TransformUp=false
   ) {enabled=false, module=UpdateMeshOrientation001}
   &ScaleSpriteSize(Scale Factor=&Vector2DFromFloat(Value=&FloatFromCurve(FloatCurve=DataInterface() {propertyValues {Curve {keys [
          {arriveTangent=1.5832, interpMode=RCIM_Cubic, time=0, value=0.5}
          {arriveTangent=0.5893, interpMode=RCIM_Cubic, leaveTangent=0.5893, tangentMode=RCTM_User, time=1, value=1}
         ]}}})))
   &UpdateMeshOrientation(
    Facing Direction=&Multiply_Vector(
     A=&SubtractVectorFromPosition(Position=$Particles.Position, Vector=$Output.SkeletalMeshLocation.SampledPosition)
     B=&Lerp_Vector(
      A=<1 1 0>
      B=<1 1 -0.03>
      Alpha=&Multiply_Float(
       A=&NormalizeFloat(
        Value=&DistanceBetweenPositions(A=Linked() {linkedVariable {name=Output.SkeletalMeshLocation.SampledPosition, type=Vector3f}}, B=$Particles.Position)
        Min=0
        Max=1
       )
       B=0.15
      )
     )
    )
    Facing Interpolation=false
    Orientation Method=::NewEnumerator1
    Override Quat=false
    TransformFacing=false
    TransformUp=false
   )
   /SetVariables(Particles.Position=&AddVectorToPosition(
     Position=$Particles.Position
     Vector=&FN_FrontEndVector3Modifier(
      Gameplay=&Multiply_Vector(
       A=$System.PiecesVelocity
       B=&Lerp_Vector(
        A=<-0.2 -0.2 0>
        B=<-1.2 -1.2 0>
        Alpha=&NormalizeFloat(
         Value=&DistanceBetweenPositions(A=Linked() {linkedVariable {name=Output.SkeletalMeshLocation.SampledPosition, type=Vector3f}}, B=$Particles.Position)
         Min=0
         Max=1
        )
       )
      )
      Front End=&Lerp_Vector(
       A=&Lerp_Vector(A=<0 0 0>, B=<0 -0.2 0>, Alpha=$System.Dive)
       B=&Lerp_Vector(A=<0 0 0>, B=<0 -1.2 0>, Alpha=$System.Dive)
       Alpha=&NormalizeFloat(
        Value=&DistanceBetweenPositions(A=Linked() {linkedVariable {name=Output.SkeletalMeshLocation.SampledPosition, type=Vector3f}}, B=$Particles.Position)
        Min=0
        Max=1
       )
      )
     )
    )) {isSetParameters=true}
   &SolveForcesAndVelocity(Clamp Velocity=false, Limit Acceleration=false)
  ]
  rendererClasses [&NiagaraMeshRendererProperties]
  renderers [
   {rendererClass=&NiagaraMeshRendererProperties, rendererIndex=0}
   {rendererClass=&NiagaraMeshRendererProperties, rendererIndex=1}
  ]
 }
 Flares {
  emitterUpdateScript [
   &EmitterState(Life Cycle Mode=::NewEnumerator1, Loop Duration=0, UseLoopDelay=false)
   &SpawnRate(SpawnRate=60, Use Spawn Probability=false)
  ]
  particleSpawnScript [
   &FN_SetSpawnGroup()
   /Niagara/Modules/Spawn/Initialization/V2/InitializeParticle(
    Color=<rgba 0.0579 0 2 1>
    Lifetime=0.3
    Sprite Rotation Angle=&RandomRangeFloat(Minimum=0, Maximum=360)
    Sprite Rotation Mode=::NewEnumerator0
    Sprite Size Mode=::NewEnumerator3
    Uniform Sprite Size=&Multiply_Float(
     A=&RandomRangeFloat(Minimum=110, Maximum=125)
     B=&MaskFloatBySpawnGroup(Masked Float=1, Passthrough Float=0.8, Spawn Group Mask=&ModuloInt(A=$Particles.SpawnGroup, B=2))
    )
    UsePositionOffset=false
    Write Material Random=true
   ) {module=InitializeParticle001}
   &CameraOffset(Camera Offset Amount=&FN_FrontEndFloatModifier(Gameplay=&Lerp_Float(A=-10, B=30, Alpha=$User.ParaGlideLeanAlpha), Front End=10)) {module=CameraOffset001}
  ]
  particleUpdateScript [
   &ParticleState()
   &SkeletalMeshLocation(
    Bone / Socket Index=$Particles.SpawnGroup
    Bone Sampling Mode=::NewEnumerator4
    Inherit Velocity=false
    MaskSpawnGroup=false
    Sampled Position Coordinate Space=::Local
    Skeletal Mesh=$User.SkeletalMesh
   ) {enabled=false, module=SkeletalMeshLocation001}
   &SkeletalMeshLocation(
    Bone / Socket Index=$Particles.SpawnGroup
    Bone Sampling Mode=::NewEnumerator4
    Inherit Velocity=false
    Interpolate Mesh Sampling=false
    MaskSpawnGroup=false
    Skeletal Mesh=$User.SkeletalMesh
   )
   /Game/Effects/Niagara/Modules/Skydiving/FN_SkydivePositionOffset_v1/FN_SkydivePositionOffset(
    Bone Orientation:Quat4f=$Output.SkeletalMeshLocation.SampledRotation
    Bone Position:Vector3f=$Output.SkeletalMeshLocation.SampledPosition
    Offset:NiagaraFloat=/Game/Effects/Niagara/DynamicInputs/Skydiving/SkydiveHandsFeetFloatModifier(Hands:NiagaraFloat=20, Feet:NiagaraFloat=45)
   ) {enabled=false}
   /Game/Effects/Niagara/Modules/Skydiving/FN_Skydive_Velocity_v4/FN_Skydive_Velocity(
    Bone Orientation:Quat4f=$Output.SkeletalMeshLocation.SampledRotation
    Bone Position:Vector3f=$Output.SkeletalMeshLocation.SampledPosition
    Dive:NiagaraFloat=$System.Dive
    FE Speed=1
    FE Spread=0.1
    SpawnGroup:NiagaraInt32=$Particles.SpawnGroup
    Write to Intrinsic Properties=false
   ) {enabled=false}
   &ScaleSpriteSize(Scale Factor=&Vector2DFromFloat(Value=&FloatFromCurve(FloatCurve=DataInterface() {propertyValues {Curve {keys [
          {arriveTangent=1.5832, interpMode=RCIM_Cubic, time=0, value=0.5}
          {arriveTangent=0.5893, interpMode=RCIM_Cubic, leaveTangent=0.5893, tangentMode=RCTM_User, time=1, value=1}
         ]}}})))
   &SolveForcesAndVelocity(Clamp Velocity=false, Limit Acceleration=false)
   &SpriteRotationRate(Rotation Rate=&MakeCustomFloatFromBool(Bool=&RandomBool(), True Float=10, False Float=-10)) {enabled=false}
   &ScaleColor(
    Scale Alpha=&FloatFromCurve(FloatCurve=DataInterface() {propertyValues {Curve {keys [
         {arriveTangent=-1.4925, interpMode=RCIM_Cubic, time=0, value=0}
         <at 0.33 1 RCIM_Cubic>
         {interpMode=RCIM_Cubic, leaveTangent=-1.9375, time=1, value=0}
        ]}}})
    Scale RGB=&VectorFromFloat(Value=&FloatFromCurve(
      FloatCurve=DataInterface() {propertyValues {Curve {keys [
          {arriveTangent=8517.9209, interpMode=RCIM_Cubic, leaveTangent=2.9045, tangentMode=RCTM_User, time=0, value=0.1}
          <at 0.33 1 RCIM_Cubic>
          {interpMode=RCIM_Cubic, leaveTangent=-4.1243, time=1, value=0.1}
         ]}}}
      Scale Curve=&Multiply_Float(A=&FNTimeOfDayScalar(Night Scale=0.5, Morning Scale=0.65, Evening Scale=0.65), B=1)
     ))
    ScaleA=true
    ScaleRGB=true
   )
   &CameraOffset(Camera Offset Amount=4) {enabled=false}
  ]
  rendererClasses [&NiagaraSpriteRendererProperties]
  renderers [{rendererClass=&NiagaraSpriteRendererProperties}]
 }
 Lights {
  emitterSpawnScript [&FN_PreserveSkydiveUserVariables() {module=PreserveUserVariables}]
  emitterUpdateScript [
   &EmitterState(Inactive Response=::NewEnumerator1, Life Cycle Mode=::NewEnumerator1, Loop Duration=0, UseLoopDelay=false)
   &SpawnBurst_Instantaneous(Spawn Count=4, Spawn Probability=HlslExpression() {hlslExpression="Emitter.LoopCount == 0"}, Use Spawn Probability=true) {module=SpawnBurst_Instantaneous001}
  ]
  particleSpawnScript [
   &FN_SetSpawnGroup()
   &InitializeParticle(
    Color Mode=::NewEnumerator0
    Lifetime=0
    Sprite Size Mode=::NewEnumerator3
    Uniform Sprite Size=&MaskFloatBySpawnGroup(Masked Float=25, Passthrough Float=30, Spawn Group Mask=&ModuloInt(A=$Particles.SpawnGroup, B=2))
    UsePositionOffset=false
    Write Material Random=true
   ) {module=InitializeParticle001}
   &CameraOffset(Camera Offset Amount=5)
   &Color(Color=$NPC.FortniteNPC.FortniteActiveTimeOfDay)
   &DynamicMaterialParameters(
    Index 0 Param 1=&MaskFloatBySpawnGroup(Masked Float=0.5, Passthrough Float=0)
    Index 0 Param 2=&FN_FrontEndFloatModifier(Gameplay=0, Front End=1)
    Module.Param0WriteEnabled=true
    Module.Param1WriteEnabled=true
    Module.Param2WriteEnabled=false
    Module.Param3WriteEnabled=false
   ) {enabled=false}
  ]
  particleUpdateScript [
   &ParticleState(Kill Particles When Lifetime Has Elapsed=false, Let Infinitely Lived Particles Die When Emitter Deactivates=true)
   &SampleSkeletalMeshSkeleton(
    Mesh:NiagaraDataInterfaceSkeletalMesh=$User.SkeletalMesh
    Sampled Rotation Coordinate Space:ENiagaraCoordinateSpace=::Local
    Specific Bone Index:NiagaraInt32=$Particles.SpawnGroup
   )
   &SkeletalMeshLocation(
    Apply Sampled Orientation to Mesh Particles=true
    Bone / Socket Index=$Particles.SpawnGroup
    Bone Sampling Mode=::NewEnumerator4
    Inherit Velocity=false
    Interpolate Mesh Sampling=false
    MaskSpawnGroup=false
    Sampled Position Offset=&MaskVectorBySpawnGroup(
     Masked Vector=&MaskVectorBySpawnGroup(Masked Vector=<10 0 0>, Passthrough Vector=<-10 0 0>)
     Passthrough Vector=&MaskVectorBySpawnGroup(Masked Vector=<-5 0 0>, Passthrough Vector=<5 0 0>, Spawn Group Mask=2)
     Spawn Group Mask=&ModuloInt(A=$Particles.SpawnGroup, B=2)
    )
    Sampled Position Offset Coordinate Space:ENiagaraMeshTransforms=::NewEnumerator3
    Skeletal Mesh=$User.SkeletalMesh
   )
  ]
  rendererClasses [&NiagaraSpriteRendererProperties]
  renderers [{rendererClass=&NiagaraSpriteRendererProperties}]
 }
 Metallic_Pieces {
  emitterSpawnScript [
   /SetVariables() {isSetParameters=true}
   /SetVariables() {isSetParameters=true}
   /SetVariables() {isSetParameters=true}
  ]
  emitterUpdateScript [
   &EmitterState(Inactive Response=::NewEnumerator1, Life Cycle Mode=::NewEnumerator1, Loop Duration=0, UseLoopDelay=false)
   &SpawnBurst_Instantaneous(Spawn Count=2, Spawn Probability=HlslExpression() {hlslExpression="Emitter.LoopCount == 0"}, Use Spawn Probability=true)
   /SetVariables(Emitter.DisksRotation:NiagaraFloat=&Add_Float(
     A=$Emitter.DisksRotation
     B=&Multiply_Float(A=$Engine.DeltaTime, B=&Lerp_Float(A=40, B=80, Alpha=$User.ParaGlideLeanAlpha))
    )) {isSetParameters=true}
   /SetVariables(Emitter.MeshWiggleFrequency:NiagaraFloat=&Lerp_Float(A=1, B=1, Alpha=$User.ParaGlideLeanAlpha)) {isSetParameters=true}
   &self:MeshesOffsetSpeed() {module=MeshesOffsetSpeed}
  ]
  particleSpawnScript [
   &FN_SetSpawnGroup(Number of Spawn Groups:NiagaraInt32=$System.SpawnGroup)
   &InitializeParticle(
    Color Mode=::NewEnumerator0
    Lifetime=0
    Mesh Scale Mode=::NewEnumerator3
    Mesh Uniform Scale=0.2
    UsePositionOffset=false
    Write Material Random=true
   ) {module=InitializeParticle001}
   &InitialMeshOrientation(
    Orientation Axis=<0 0 1>
    Orientation Vector=<0 0 1>
    Rotation=&MakeVector(X=0, Y=0, Z=&RandomRangeFloat(Minimum=0, Maximum=1))
    Rotation Coordinate Space=::NewEnumerator0
    Use Orientation Vector=false
    Use Rotation Vector=true
   )
   &DynamicMaterialParameters(
    Index 0 Param 1=&RandomRangeFloat(Minimum=0, Maximum=10)
    Index 0 Param 2=$Emitter.MeshWiggleFrequency
    Index 0 Param 3=&RandomRangeFloat(Minimum=0.7, Maximum=1)
    Index 0 Param 4=&Lerp_Float(A=0.4, B=1.1, Alpha=$User.ParaGlideLeanAlpha)
    Index 1 Param 1=&RandomRangeFloat(Minimum=11, Maximum=20)
    Index 1 Param 2=$Emitter.MeshWiggleFrequency
    Index 1 Param 3=&RandomRangeFloat(Minimum=0.7, Maximum=1)
    Index 1 Param 4=&Lerp_Float(A=0.4, B=1.1, Alpha=$User.ParaGlideLeanAlpha)
    Index 2 Param 1=&RandomRangeFloat(Minimum=21, Maximum=30)
    Index 2 Param 2=$Emitter.MeshWiggleFrequency
    Index 2 Param 3=&RandomRangeFloat(Minimum=0.7, Maximum=1)
    Index 2 Param 4=&Lerp_Float(A=0.4, B=1.1, Alpha=$User.ParaGlideLeanAlpha)
    Module.Param0WriteEnabled=true
    Module.Param0WriteEnabled1=true
    Module.Param0WriteEnabled2=true
    Module.Param1WriteEnabled=true
    Module.Param1WriteEnabled1=true
    Module.Param1WriteEnabled2=true
    Module.Param2WriteEnabled=true
    Module.Param2WriteEnabled1=true
    Module.Param2WriteEnabled2=true
    Module.Param3WriteEnabled=true
    Module.Param3WriteEnabled1=true
    Module.Param3WriteEnabled2=true
    Write Parameter Index 1=true
    Write Parameter Index 2=true
   )
   /SetVariables() {enabled=false, isSetParameters=true}
  ]
  particleUpdateScript [
   &ParticleState(Kill Particles When Lifetime Has Elapsed=false, Let Infinitely Lived Particles Die When Emitter Deactivates=true)
   &DynamicMaterialParameters(
    Index 0 Param 2=$Emitter.MeshWiggleFrequency
    Index 0 Param 4=&Lerp_Float(A=0.4, B=1.1, Alpha=$User.ParaGlideLeanAlpha)
    Index 1 Param 2=$Emitter.MeshWiggleFrequency
    Index 1 Param 4=&Lerp_Float(A=0.4, B=1.1, Alpha=$User.ParaGlideLeanAlpha)
    Index 2 Param 2=$Emitter.MeshWiggleFrequency
    Index 2 Param 3=&Lerp_Float(A=0.4, B=1.1, Alpha=$User.ParaGlideLeanAlpha)
    Module.Param0WriteEnabled=false
    Module.Param0WriteEnabled1=false
    Module.Param0WriteEnabled2=false
    Module.Param1WriteEnabled=true
    Module.Param1WriteEnabled1=true
    Module.Param1WriteEnabled2=true
    Module.Param2WriteEnabled=false
    Module.Param2WriteEnabled1=false
    Module.Param2WriteEnabled2=true
    Module.Param3WriteEnabled=true
    Module.Param3WriteEnabled1=true
    Module.Param3WriteEnabled2=true
    Write Parameter Index 1=true
    Write Parameter Index 2=true
   ) {module=DynamicMaterialParameters001}
   &SkeletalMeshLocation(
    Bone / Socket Index=$Particles.SpawnGroup
    Bone Sampling Mode=::NewEnumerator4
    Inherit Velocity=false
    Interpolate Mesh Sampling=false
    MaskSpawnGroup=false
    Sampled Position Offset=&FN_FrontEndVector3Modifier(
     Gameplay=&Multiply_VectorByFloat(Vector=$System.PiecesVelocity, Float=$System.PiecesOffsetDistance)
     Front End=&Multiply_VectorByFloat(Vector=&Lerp_Vector(A=<0 0 -1>, B=<0 1 -1>, Alpha=$User.ParaGlideLeanAlpha), Float=$System.PiecesOffsetDistance)
    )
    Sampled Position Offset Coordinate Space:ENiagaraMeshTransforms=::NewEnumerator2
    Skeletal Mesh=$User.SkeletalMesh
   )
   /SetVariables(
    Particles.Position001:NiagaraPosition=&AddVectorToPosition(
     Position=$Particles.Position
     Vector=&FN_FrontEndVector3Modifier(
      Gameplay=&Multiply_Vector(
       A=&AddVectorToPosition(
        Position=&Lerp_Vector(
         A=<0 0 0>
         B=<-20 -20 -20>
         Alpha=&NormalizeFloat(Value=&Multiply_Float(A=$System.FN_Integrator001.Value, B=$User.ParaGlideLeanAlpha), Min=0, Max=1)
        )
        Vector=&Multiply_VectorByFloat(Vector=<0.3 0.3 0.3>, Float=$Emitter.MeshesInitialOffsetSpeed)
       )
       B=$System.PiecesVelocity
      )
      Front End=&Multiply_Vector(A=&Lerp_Vector(A=<0 0 0>, B=<-1 -1 -1>, Alpha=$System.Dive), B=<0 -1 -1>)
     )
    )
    Particles.Position002:NiagaraPosition=&AddVectorToPosition(
     Position=$Particles.Position
     Vector=&FN_FrontEndVector3Modifier(
      Gameplay=&Multiply_Vector(
       A=&AddVectorToPosition(
        Position=&Lerp_Vector(
         A=<0 0 0>
         B=<-35 -35 -40>
         Alpha=&NormalizeFloat(Value=&Multiply_Float(A=$System.FN_Integrator001.Value, B=$User.ParaGlideLeanAlpha), Min=0, Max=1)
        )
        Vector=&Multiply_VectorByFloat(Vector=<0.6 0.6 0.6>, Float=$Emitter.MeshesInitialOffsetSpeed)
       )
       B=$System.PiecesVelocity
      )
      Front End=&Multiply_Vector(A=&Lerp_Vector(A=<0 0 0>, B=<0 -1 1>, Alpha=$System.Dive), B=<0 8 8>)
     )
    )
    Particles.Position003:NiagaraPosition=&AddVectorToPosition(
     Position=$Particles.Position
     Vector=&FN_FrontEndVector3Modifier(
      Gameplay=&Multiply_Vector(
       A=&AddVectorToPosition(
        Position=&Lerp_Vector(
         A=<0 0 0>
         B=<-55 -55 -60>
         Alpha=&NormalizeFloat(Value=&Multiply_Float(A=$System.FN_Integrator001.Value, B=$User.ParaGlideLeanAlpha), Min=0, Max=1)
        )
        Vector=&Multiply_VectorByFloat(Vector=<1 1 1>, Float=$Emitter.MeshesInitialOffsetSpeed)
       )
       B=$System.PiecesVelocity
      )
      Front End=&Multiply_Vector(A=&Lerp_Vector(A=<0 0 0>, B=<0 -1 1>, Alpha=$System.Dive), B=<0 15 15>)
     )
    )
   ) {isSetParameters=true}
   &SkeletalMeshLocation(
    Bone / Socket Index=$Particles.SpawnGroup
    Bone Sampling Mode=::NewEnumerator4
    Inherit Velocity=false
    Interpolate Mesh Sampling=false
    MaskSpawnGroup=false
    Sampled Position Offset=&Multiply_VectorByFloat(
     Vector=$System.PiecesVelocity
     Float=&Add_Float(
      A=$System.PiecesOffsetDistance
      B=&Sine(
       Normalized Angle=&Add_Float(A=$System.FN_Integrator.Value, B=&RandomRangeFloat(Minimum=0, Maximum=1))
       Period=1
       Scale=1
       Bias=0
      )
     )
    )
    Sampled Position Offset Coordinate Space:ENiagaraMeshTransforms=::NewEnumerator2
    Skeletal Mesh=$User.SkeletalMesh
   ) {enabled=false, module=SkeletalMeshLocation001}
   /SetVariables() {isSetParameters=true}
   /SetVariables(Particles.MeshOrientation=&MultiplyQuaternion(
     Quaternion A=&FindQuatBetweenTwoVectors(
      From Vector=&MakeVector(
       X=&Sine(
        Normalized Angle=&Add_Float(A=$System.FN_Integrator.Value, B=&RandomRangeFloat(Minimum=0, Maximum=1))
        Period=&MakeCustomFloatFromBool(Bool=&RandomBool(), True Float=1.5, False Float=-1.5)
        Scale=0.15
        Bias=0
       )
       Y=&Cosine(
        Normalized Angle=&Add_Float(A=$System.FN_Integrator.Value, B=&RandomRangeFloat(Minimum=0, Maximum=1))
        Period=&MakeCustomFloatFromBool(Bool=&RandomBool(), True Float=1.5, False Float=-1.5)
        Scale=0.15
        Bias=0
       )
       Z=&RandomRangeFloat(Minimum=4, Maximum=5)
      )
      To Vector=<0 0 1>
     )
     Quaternion B=$Particles.Initial.MeshOrientation
    )) {enabled=false, isSetParameters=true}
   /SetVariables(Particles.MeshOrientation=&MultiplyQuaternion(
     Quaternion A=&FindQuatBetweenTwoVectors(
      From Vector=<0 0 -1>
      To Vector=&FN_FrontEndVector3Modifier(Gameplay=$System.PiecesVelocity, Front End=&Lerp_Vector(A=<0 0 -1>, B=<0 1 -1>, Alpha=$User.ParaGlideLeanAlpha))
     )
     Quaternion B=$Particles.MeshOrientation
    )) {isSetParameters=true}
   /SetVariables(Particles.MeshOrientation=&MultiplyQuaternion(
     Quaternion A=$Particles.MeshOrientation
     Quaternion B=&MakeQuatFromAxisAngle(Axis=<0 0 1>, AngleInDegrees=&Multiply_Float(A=$Emitter.DisksRotation, B=1))
    )) {isSetParameters=true}
   &SolveForcesAndVelocity(Clamp Velocity=false, Limit Acceleration=false)
  ]
  rendererClasses [&NiagaraMeshRendererProperties]
  renderers [
   {rendererClass=&NiagaraMeshRendererProperties, rendererIndex=0}
   {rendererClass=&NiagaraMeshRendererProperties, rendererIndex=1}
   {rendererClass=&NiagaraMeshRendererProperties, rendererIndex=2}
  ]
 }
 Ribbons001 {
  bEnabled=false
  emitterUpdateScript [
   &EmitterState(Life Cycle Mode=::NewEnumerator1, Loop Duration=0, UseLoopDelay=false)
   /Game/Effects/Niagara/Modules/Skydiving/FN_DivingAttributes/FN_EmitterDivingAttributes_v1(Bone Index=1, SkeletalMesh:NiagaraDataInterfaceSkeletalMesh=$User.SkeletalMesh)
   &FN_SpawnRibbon_v2(Number of Ribbons=4, Spawn Mode=::NewEnumerator0, SpawnRate=256)
  ]
  particleSpawnScript [
   &FN_SetSpawnGroup()
   &InitializeParticle(
    Color Mode=::NewEnumerator0
    Lifetime=0
    Mesh Scale Mode=::NewEnumerator3
    Mesh Uniform Scale=0.2
    UsePositionOffset=false
    Write Material Random=true
   ) {enabled=false}
   &InitialMeshOrientation(Mesh Orientation Mode=::NewEnumerator7, Rotation=<0 0.25 0>, Use Rotation Vector=true) {enabled=false}
   &FN_InitializeRibbon_v3(
    Custom Offset=true
    Enable Skeletal Position=true
    Lifetime:NiagaraFloat=&MaskFloatBySpawnGroup(
     Masked Float=&Lerp_Float(A=0.16, B=0.08, Alpha=$User.ParaGlideLeanAlpha)
     Passthrough Float=&Lerp_Float(A=0.07, B=0.035, Alpha=$User.ParaGlideLeanAlpha)
     Spawn Group Mask=&ModuloInt(A=$Particles.SpawnGroup, B=2)
    )
    Local Vector Offset:Vector3f=<0 1 0>
    Ribbon Width:NiagaraFloat=&MaskFloatBySpawnGroup(Masked Float=32, Passthrough Float=18, Spawn Group Mask=&ModuloInt(A=$Particles.SpawnGroup, B=2))
   )
   &AddVelocity(
    Scale Added Velocity=&FN_FrontEndVector3Modifier(
     Gameplay=&Lerp_Vector(A=<1 1 1>, B=<4 4 1>, Alpha=$System.Dive)
     Front End=&Lerp_Vector(A=<0 0 1>, B=<3 3 1.8>, Alpha=$System.Dive)
    )
    Velocity=&FN_FrontEndVector3Modifier(Gameplay=<0 0 100>, Front End=<0 -600 1000>)
   )
   &DynamicMaterialParameters(
    Index 0 Param 1=&Multiply_Float(A=&MakeFloatFromInt(Int=$Particles.SpawnGroup), B=0.2)
    Index 0 Param 2=&FN_FrontEndFloatModifier(Gameplay=1, Front End=0.8)
    Index 0 Param 3=&Multiply_Float(A=&MakeFloatFromInt(Int=$Particles.SpawnGroup), B=0.4)
    Index 0 Param 4=&MaskFloatBySpawnGroup(
     Masked Float=1
     Passthrough Float=0.5
     Spawn Group Mask=&ModuloInt(A=&SpawnGroup(), B=2)
    )
    Module.Param0WriteEnabled=true
    Module.Param1WriteEnabled=true
    Module.Param2WriteEnabled=true
    Module.Param3WriteEnabled=true
   )
   &SampleSkeletalMesh(
    AdditionalYawPitchRoll=false
    Bone Sampling Mode:ENiagaraBoneSamplingMode=::NewEnumerator4
    Skeletal Mesh:NiagaraDataInterfaceSkeletalMesh=$User.SkeletalMesh
   ) {module=SampleSkeletalMesh001}
   &SkeletalMeshLocation(
    Apply Sampled Orientation to Mesh Particles=true
    Bone / Socket Index=$Particles.SpawnGroup
    Bone Sampling Mode=::NewEnumerator4
    Inherit Velocity=false
    Interpolate Mesh Sampling=false
    MaskSpawnGroup=false
    Sampled Position Offset=&MaskVectorBySpawnGroup(Masked Vector=<0 0 0>, Passthrough Vector=<0 0 4>, Spawn Group Mask=&ModuloInt(A=$Particles.SpawnGroup, B=2))
    Sampled Position Offset Coordinate Space:ENiagaraMeshTransforms=::NewEnumerator2
    Skeletal Mesh=$User.SkeletalMesh
   )
   &Color(Color=<rgba 1 1 1 0.1>)
  ]
  particleUpdateScript [
   &InheritSourceMovement(Applied Position Scale=<0.5 0.5 1>, ApplyForce=false, ApplyPosition=true, ApplyPositionFalloff=false) {enabled=false}
   &SolveForcesAndVelocity(Clamp Velocity=false, Limit Acceleration=false)
   &FN_UpdateRibbon(Detach Locked Particle=true, Enable Skeletal Position=true)
   &ScaleColor(
    Scale Alpha=&FloatFromCurve(
     FloatCurve=DataInterface() {propertyValues {Curve {keys [
         {arriveTangent=-0.3544, interpMode=RCIM_Cubic, leaveTangent=-0.3544, tangentMode=RCTM_User, time=0.3, value=-1}
         {arriveTangent=0.3316, interpMode=RCIM_Cubic, leaveTangent=0.3316, tangentMode=RCTM_User, time=1, value=1}
        ]}}}
     Scale Curve=10
    )
    Scale RGB=&FN_Select_Vector_from_Time_Of_Day(
     Morning=&MakeVectorFromLinearColorRGB(Color=&ColorFromCurve(ColorCurve=DataInterface() {propertyValues {
         AlphaCurve {keys [<at 0 1 RCIM_Cubic>]}
         BlueCurve {keys [<at 0.1134 0.15>,<at 0.4335 0.7>,<at 0.9804 1.4742>]}
         GreenCurve {keys [<at 0.1134 0>,<at 0.4335 0>,<at 0.9804 0.03>]}
         RedCurve {keys [<at 0.1134 0.0321>,<at 0.4335 0.3188>,<at 0.9804 3>]}
        }}))
     Day=&MakeVectorFromLinearColorRGB(Color=&ColorFromCurve(ColorCurve=DataInterface() {propertyValues {
         AlphaCurve {keys [<at 0 1 RCIM_Cubic>]}
         BlueCurve {keys [<at 0.1134 0.25>,<at 0.4335 0.5>,<at 0.9804 0.9828>]}
         GreenCurve {keys [<at 0.1134 0>,<at 0.4335 0>,<at 0.9804 0.02>]}
         RedCurve {keys [<at 0.1134 0.0534>,<at 0.4335 0.2277>,<at 0.9804 2>]}
        }}))
     Evening=&MakeVectorFromLinearColorRGB(Color=&ColorFromCurve(ColorCurve=DataInterface() {propertyValues {
         AlphaCurve {keys [<at 0 1 RCIM_Cubic>]}
         BlueCurve {keys [<at 0.1134 0.15>,<at 0.4335 0.7>,<at 0.9804 1.4742>]}
         GreenCurve {keys [<at 0.1134 0>,<at 0.4335 0>,<at 0.9804 0.03>]}
         RedCurve {keys [<at 0.1134 0.0321>,<at 0.4335 0.3188>,<at 0.9804 3>]}
        }}))
     Night=&MakeVectorFromLinearColorRGB(Color=&ColorFromCurve(ColorCurve=DataInterface() {propertyValues {
         AlphaCurve {keys [<at 0 1 RCIM_Cubic>]}
         BlueCurve {keys [<at 0.1134 0.1>,<at 0.4335 0.45>,<at 0.9804 0.4914>]}
         GreenCurve {keys [<at 0.1134 0>,<at 0.4335 0>,<at 0.9804 0.01>]}
         RedCurve {keys [<at 0.1134 0.0214>,<at 0.4335 0.2049>,<at 0.9804 1>]}
        }}))
    )
    ScaleA=true
    ScaleRGB=true
   )
   &ScaleColor(
    Scale Alpha=&FloatFromCurve(
     FloatCurve=DataInterface() {propertyValues {Curve {keys [
         {arriveTangent=-0.3544, interpMode=RCIM_Cubic, leaveTangent=-0.3544, tangentMode=RCTM_User, time=0.3, value=-1}
         {arriveTangent=3.1582, interpMode=RCIM_Cubic, leaveTangent=3.1582, tangentMode=RCTM_User, time=1, value=1}
        ]}}}
     Scale Curve=10
    )
    Scale RGB=&Multiply_VectorByFloat(
     Vector=&FN_Select_Vector_from_Time_Of_Day(
      Morning=&MakeVectorFromLinearColorRGB(Color=&ColorFromCurve(ColorCurve=DataInterface() {propertyValues {
          AlphaCurve {keys [<at 0 1 RCIM_Cubic>]}
          BlueCurve {keys [<at 0.0108 2 RCIM_Cubic>,<at 0.2828 1.4>,<at 0.6245 1>,<at 0.9795 0.3>]}
          GreenCurve {keys [<at 0.0108 0.8667 RCIM_Cubic>,<at 0.2828 0.028>,<at 0.6245 0>,<at 0.9795 0.0045>]}
          RedCurve {keys [<at 0.0108 0 RCIM_Cubic>,<at 0.2828 0.0509>,<at 0.6245 0.5>,<at 0.9795 0.103>]}
         }}))
      Day=&MakeVectorFromLinearColorRGB(Color=&ColorFromCurve(ColorCurve=DataInterface() {propertyValues {
          AlphaCurve {keys [<at 0 1 RCIM_Cubic>]}
          BlueCurve {keys [<at 0.0163 3 RCIM_Cubic>,<at 0.2828 3>,<at 0.6245 1>,<at 0.9795 0.3>]}
          GreenCurve {keys [<at 0.0163 1.3 RCIM_Cubic>,<at 0.2828 0.06>,<at 0.6245 0>,<at 0.9795 0.0045>]}
          RedCurve {keys [<at 0.0163 0 RCIM_Cubic>,<at 0.2828 0.109>,<at 0.6245 0.5>,<at 0.9795 0.103>]}
         }}))
      Evening=&MakeVectorFromLinearColorRGB(Color=&ColorFromCurve(ColorCurve=DataInterface() {propertyValues {
          AlphaCurve {keys [<at 0 1 RCIM_Cubic>]}
          BlueCurve {keys [<at 0.0108 2 RCIM_Cubic>,<at 0.2828 1.4>,<at 0.6245 1>,<at 0.9795 0.3>]}
          GreenCurve {keys [<at 0.0108 0.8667 RCIM_Cubic>,<at 0.2828 0.028>,<at 0.6245 0>,<at 0.9795 0.0045>]}
          RedCurve {keys [<at 0.0108 0 RCIM_Cubic>,<at 0.2828 0.0509>,<at 0.6245 0.5>,<at 0.9795 0.103>]}
         }}))
      Night=&MakeVectorFromLinearColorRGB(Color=&ColorFromCurve(ColorCurve=DataInterface() {propertyValues {
          AlphaCurve {keys [<at 0 1 RCIM_Cubic>]}
          BlueCurve {keys [<at 0.0108 1.2 RCIM_Cubic>,<at 0.2789 1.4>,<at 0.6245 1>,<at 0.9795 0.3>]}
          GreenCurve {keys [<at 0.0108 0.656 RCIM_Cubic>,<at 0.2789 0.14>,<at 0.6245 0.05>,<at 0.9795 0.024>]}
          RedCurve {keys [<at 0.0108 0.24 RCIM_Cubic>,<at 0.2789 0.3122>,<at 0.6245 0.525>,<at 0.9795 0.116>]}
         }}))
     )
     Float=&FN_FrontEndFloatModifier(Gameplay=1, Front End=1.9)
    )
    ScaleA=true
    ScaleRGB=true
   ) {enabled=false, module=ScaleColor002}
   &ScaleColor(
    Scale Alpha=&FloatFromCurve(
     FloatCurve=DataInterface() {propertyValues {Curve {keys [
         {arriveTangent=-0.3544, interpMode=RCIM_Cubic, leaveTangent=-0.3544, tangentMode=RCTM_User, time=0.3, value=-1}
         {arriveTangent=3.1582, interpMode=RCIM_Cubic, leaveTangent=3.1582, tangentMode=RCTM_User, time=1, value=1}
        ]}}}
     Scale Curve=10
    )
    Scale RGB=&FN_Select_Vector_from_Time_Of_Day(
     Morning=&MakeVectorFromLinearColorRGB(Color=&ColorFromCurve(ColorCurve=DataInterface() {propertyValues {
         AlphaCurve {keys [<at 0 1 RCIM_Cubic>]}
         BlueCurve {keys [<at 0 1.7 RCIM_Cubic>,<at 0.6245 1>,<at 0.9795 0.3>]}
         GreenCurve {keys [<at 0 1.292 RCIM_Cubic>,<at 0.6245 0>,<at 0.9795 0.0047>]}
         RedCurve {keys [<at 0 0.17 RCIM_Cubic>,<at 0.6245 0.4907>,<at 0.9795 0.1031>]}
        }}))
     Day=&MakeVectorFromLinearColorRGB(Color=&ColorFromCurve(ColorCurve=DataInterface() {propertyValues {
         AlphaCurve {keys [<at 0 1 RCIM_Cubic>]}
         BlueCurve {keys [<at 0 3 RCIM_Cubic>,<at 0.6245 1>,<at 0.9795 0.3>]}
         GreenCurve {keys [<at 0 2.28 RCIM_Cubic>,<at 0.6245 0>,<at 0.9795 0.0047>]}
         RedCurve {keys [<at 0 0.3 RCIM_Cubic>,<at 0.6245 0.4907>,<at 0.9795 0.1031>]}
        }}))
     Evening=&MakeVectorFromLinearColorRGB(Color=&ColorFromCurve(ColorCurve=DataInterface() {propertyValues {
         AlphaCurve {keys [<at 0 1 RCIM_Cubic>]}
         BlueCurve {keys [<at 0 1.7 RCIM_Cubic>,<at 0.6245 1>,<at 0.9795 0.3>]}
         GreenCurve {keys [<at 0 1.292 RCIM_Cubic>,<at 0.6245 0>,<at 0.9795 0.0047>]}
         RedCurve {keys [<at 0 0.17 RCIM_Cubic>,<at 0.6245 0.4907>,<at 0.9795 0.1031>]}
        }}))
     Night=&MakeVectorFromLinearColorRGB(Color=&ColorFromCurve(ColorCurve=DataInterface() {propertyValues {
         AlphaCurve {keys [<at 0 1 RCIM_Cubic>]}
         BlueCurve {keys [<at 0 1.4 RCIM_Cubic>,<at 0.6245 1>,<at 0.9795 0.3>]}
         GreenCurve {keys [<at 0 1.273 RCIM_Cubic>,<at 0.6245 0>,<at 0.9795 0.0047>]}
         RedCurve {keys [<at 0 0.9293 RCIM_Cubic>,<at 0.6245 0.4907>,<at 0.9795 0.1031>]}
        }}))
    )
    ScaleA=true
    ScaleRGB=true
   ) {enabled=false, module=ScaleColor001}
   /SetVariables(Particles.RibbonWidth=&Multiply_Float(A=$Particles.RibbonWidth, B=&Lerp_Float(A=1, B=0.9, Alpha=$User.ParaGlideLeanAlpha))) {enabled=false, isSetParameters=true, module=SetVariables_#5}
   &ScaleRibbonWidth(Ribbon Width Scale=&FloatFromCurve(FloatCurve=DataInterface() {propertyValues {Curve {keys [
         {arriveTangent=-0.1351, interpMode=RCIM_Cubic, leaveTangent=-0.1351, tangentMode=RCTM_User, time=0.1, value=1}
         {arriveTangent=-0.6647, interpMode=RCIM_Cubic, leaveTangent=-0.6647, tangentMode=RCTM_User, time=1, value=0.6}
        ]}}}))
  ]
  rendererClasses [&NiagaraRibbonRendererProperties]
  renderers [{rendererClass=&NiagaraRibbonRendererProperties}]
 }
 Ribbons002 {
  bEnabled=false
  emitterUpdateScript [
   &EmitterState(Life Cycle Mode=::NewEnumerator1, Loop Duration=0, UseLoopDelay=false)
   &FN_EmitterDivingAttributes_v1(Bone Index=1, SkeletalMesh:NiagaraDataInterfaceSkeletalMesh=$User.SkeletalMesh)
   &FN_SpawnRibbon_v2(Number of Ribbons=4, Spawn Mode=::NewEnumerator0, SpawnRate=256)
  ]
  particleSpawnScript [
   &FN_SetSpawnGroup()
   &FN_InitializeRibbon_v3(
    Enable Skeletal Position=true
    Lifetime:NiagaraFloat=&MaskFloatBySpawnGroup(
     Masked Float=&Lerp_Float(A=0.16, B=0.1, Alpha=$User.ParaGlideLeanAlpha)
     Passthrough Float=&Lerp_Float(A=0.08, B=0.06, Alpha=$User.ParaGlideLeanAlpha)
     Spawn Group Mask=&ModuloInt(A=$Particles.SpawnGroup, B=2)
    )
    Ribbon Width:NiagaraFloat=&MaskFloatBySpawnGroup(Masked Float=32, Passthrough Float=18, Spawn Group Mask=&ModuloInt(A=$Particles.SpawnGroup, B=2))
   )
   &AddVelocity(
    Scale Added Velocity=&FN_FrontEndVector3Modifier(
     Gameplay=&Lerp_Vector(A=<1 1 1>, B=<4 4 1>, Alpha=$System.Dive)
     Front End=&Lerp_Vector(A=<0 0 1>, B=<3 3 1.8>, Alpha=$System.Dive)
    )
    Velocity=&FN_FrontEndVector3Modifier(Gameplay=<0 0 100>, Front End=<0 -600 1000>)
   )
   &DynamicMaterialParameters(
    Index 0 Param 1=&Multiply_Float(A=&MakeFloatFromInt(Int=$Particles.SpawnGroup), B=0.2)
    Index 0 Param 2=&FN_FrontEndFloatModifier(Gameplay=1, Front End=0.8)
    Index 0 Param 3=&Multiply_Float(A=&MakeFloatFromInt(Int=$Particles.SpawnGroup), B=0.4)
    Index 0 Param 4=&MaskFloatBySpawnGroup(
     Masked Float=1
     Passthrough Float=0.5
     Spawn Group Mask=&ModuloInt(A=&SpawnGroup(), B=2)
    )
    Module.Param0WriteEnabled=true
    Module.Param1WriteEnabled=true
    Module.Param2WriteEnabled=true
    Module.Param3WriteEnabled=true
   )
   &SampleSkeletalMesh(
    AdditionalYawPitchRoll=false
    Bone Sampling Mode:ENiagaraBoneSamplingMode=::NewEnumerator4
    Skeletal Mesh:NiagaraDataInterfaceSkeletalMesh=$User.SkeletalMesh
   ) {module=SampleSkeletalMesh001}
   &SkeletalMeshLocation(
    Apply Sampled Orientation to Mesh Particles=true
    Bone / Socket Index=$Particles.SpawnGroup
    Bone Sampling Mode=::NewEnumerator4
    Inherit Velocity=false
    Interpolate Mesh Sampling=false
    MaskSpawnGroup=false
    Sampled Position Offset=&MaskVectorBySpawnGroup(Masked Vector=<0 0 0>, Passthrough Vector=<0 0 4>, Spawn Group Mask=&ModuloInt(A=$Particles.SpawnGroup, B=2))
    Sampled Position Offset Coordinate Space:ENiagaraMeshTransforms=::NewEnumerator2
    Skeletal Mesh=$User.SkeletalMesh
   )
   &Color(Color=<rgba 1 1 1 0.1>)
  ]
  particleUpdateScript [
   &InheritSourceMovement(Applied Position Scale=<0.5 0.5 1>, ApplyForce=false, ApplyPosition=true, ApplyPositionFalloff=false) {enabled=false}
   &SolveForcesAndVelocity(Clamp Velocity=false, Limit Acceleration=false)
   &FN_UpdateRibbon(Detach Locked Particle=true, Enable Skeletal Position=true)
   &ScaleColor(
    Scale Alpha=&FloatFromCurve(
     FloatCurve=DataInterface() {propertyValues {Curve {keys [
         {arriveTangent=-0.3544, interpMode=RCIM_Cubic, leaveTangent=-0.3544, tangentMode=RCTM_User, time=0.3, value=-1}
         {arriveTangent=0.3316, interpMode=RCIM_Cubic, leaveTangent=0.3316, tangentMode=RCTM_User, time=1, value=1}
        ]}}}
     Scale Curve=10
    )
    Scale RGB=&FN_Select_Vector_from_Time_Of_Day(
     Morning=&MakeVectorFromLinearColorRGB(Color=&ColorFromCurve(ColorCurve=DataInterface() {propertyValues {
         AlphaCurve {keys [<at 0 1 RCIM_Cubic>]}
         BlueCurve {keys [<at 0.1134 0.15>,<at 0.4335 0.7>,<at 0.9804 1.4742>]}
         GreenCurve {keys [<at 0.1134 0>,<at 0.4335 0>,<at 0.9804 0.03>]}
         RedCurve {keys [<at 0.1134 0.0321>,<at 0.4335 0.3188>,<at 0.9804 3>]}
        }}))
     Day=&MakeVectorFromLinearColorRGB(Color=&ColorFromCurve(ColorCurve=DataInterface() {propertyValues {
         AlphaCurve {keys [<at 0 1 RCIM_Cubic>]}
         BlueCurve {keys [<at 0.1134 0.25>,<at 0.4335 0.5>,<at 0.9804 0.9828>]}
         GreenCurve {keys [<at 0.1134 0>,<at 0.4335 0>,<at 0.9804 0.02>]}
         RedCurve {keys [<at 0.1134 0.0534>,<at 0.4335 0.2277>,<at 0.9804 2>]}
        }}))
     Evening=&MakeVectorFromLinearColorRGB(Color=&ColorFromCurve(ColorCurve=DataInterface() {propertyValues {
         AlphaCurve {keys [<at 0 1 RCIM_Cubic>]}
         BlueCurve {keys [<at 0.1134 0.15>,<at 0.4335 0.7>,<at 0.9804 1.4742>]}
         GreenCurve {keys [<at 0.1134 0>,<at 0.4335 0>,<at 0.9804 0.03>]}
         RedCurve {keys [<at 0.1134 0.0321>,<at 0.4335 0.3188>,<at 0.9804 3>]}
        }}))
     Night=&MakeVectorFromLinearColorRGB(Color=&ColorFromCurve(ColorCurve=DataInterface() {propertyValues {
         AlphaCurve {keys [<at 0 1 RCIM_Cubic>]}
         BlueCurve {keys [<at 0.1134 0.1>,<at 0.4335 0.45>,<at 0.9804 0.4914>]}
         GreenCurve {keys [<at 0.1134 0>,<at 0.4335 0>,<at 0.9804 0.01>]}
         RedCurve {keys [<at 0.1134 0.0214>,<at 0.4335 0.2049>,<at 0.9804 1>]}
        }}))
    )
    ScaleA=true
    ScaleRGB=true
   )
   &ScaleColor(
    Scale Alpha=&FloatFromCurve(
     FloatCurve=DataInterface() {propertyValues {Curve {keys [
         {arriveTangent=-0.3544, interpMode=RCIM_Cubic, leaveTangent=-0.3544, tangentMode=RCTM_User, time=0.3, value=-1}
         {arriveTangent=3.1582, interpMode=RCIM_Cubic, leaveTangent=3.1582, tangentMode=RCTM_User, time=1, value=1}
        ]}}}
     Scale Curve=10
    )
    Scale RGB=&Multiply_VectorByFloat(
     Vector=&FN_Select_Vector_from_Time_Of_Day(
      Morning=&MakeVectorFromLinearColorRGB(Color=&ColorFromCurve(ColorCurve=DataInterface() {propertyValues {
          AlphaCurve {keys [<at 0 1 RCIM_Cubic>]}
          BlueCurve {keys [<at 0.0108 2 RCIM_Cubic>,<at 0.2828 1.4>,<at 0.6245 1>,<at 0.9795 0.3>]}
          GreenCurve {keys [<at 0.0108 0.8667 RCIM_Cubic>,<at 0.2828 0.028>,<at 0.6245 0>,<at 0.9795 0.0045>]}
          RedCurve {keys [<at 0.0108 0 RCIM_Cubic>,<at 0.2828 0.0509>,<at 0.6245 0.5>,<at 0.9795 0.103>]}
         }}))
      Day=&MakeVectorFromLinearColorRGB(Color=&ColorFromCurve(ColorCurve=DataInterface() {propertyValues {
          AlphaCurve {keys [<at 0 1 RCIM_Cubic>]}
          BlueCurve {keys [<at 0.0163 3 RCIM_Cubic>,<at 0.2828 3>,<at 0.6245 1>,<at 0.9795 0.3>]}
          GreenCurve {keys [<at 0.0163 1.3 RCIM_Cubic>,<at 0.2828 0.06>,<at 0.6245 0>,<at 0.9795 0.0045>]}
          RedCurve {keys [<at 0.0163 0 RCIM_Cubic>,<at 0.2828 0.109>,<at 0.6245 0.5>,<at 0.9795 0.103>]}
         }}))
      Evening=&MakeVectorFromLinearColorRGB(Color=&ColorFromCurve(ColorCurve=DataInterface() {propertyValues {
          AlphaCurve {keys [<at 0 1 RCIM_Cubic>]}
          BlueCurve {keys [<at 0.0108 2 RCIM_Cubic>,<at 0.2828 1.4>,<at 0.6245 1>,<at 0.9795 0.3>]}
          GreenCurve {keys [<at 0.0108 0.8667 RCIM_Cubic>,<at 0.2828 0.028>,<at 0.6245 0>,<at 0.9795 0.0045>]}
          RedCurve {keys [<at 0.0108 0 RCIM_Cubic>,<at 0.2828 0.0509>,<at 0.6245 0.5>,<at 0.9795 0.103>]}
         }}))
      Night=&MakeVectorFromLinearColorRGB(Color=&ColorFromCurve(ColorCurve=DataInterface() {propertyValues {
          AlphaCurve {keys [<at 0 1 RCIM_Cubic>]}
          BlueCurve {keys [<at 0.0108 1.2 RCIM_Cubic>,<at 0.2789 1.4>,<at 0.6245 1>,<at 0.9795 0.3>]}
          GreenCurve {keys [<at 0.0108 0.656 RCIM_Cubic>,<at 0.2789 0.14>,<at 0.6245 0.05>,<at 0.9795 0.024>]}
          RedCurve {keys [<at 0.0108 0.24 RCIM_Cubic>,<at 0.2789 0.3122>,<at 0.6245 0.525>,<at 0.9795 0.116>]}
         }}))
     )
     Float=&FN_FrontEndFloatModifier(Gameplay=1, Front End=1.9)
    )
    ScaleA=true
    ScaleRGB=true
   ) {enabled=false, module=ScaleColor002}
   &ScaleColor(
    Scale Alpha=&FloatFromCurve(
     FloatCurve=DataInterface() {propertyValues {Curve {keys [
         {arriveTangent=-0.3544, interpMode=RCIM_Cubic, leaveTangent=-0.3544, tangentMode=RCTM_User, time=0.3, value=-1}
         {arriveTangent=3.1582, interpMode=RCIM_Cubic, leaveTangent=3.1582, tangentMode=RCTM_User, time=1, value=1}
        ]}}}
     Scale Curve=10
    )
    Scale RGB=&FN_Select_Vector_from_Time_Of_Day(
     Morning=&MakeVectorFromLinearColorRGB(Color=&ColorFromCurve(ColorCurve=DataInterface() {propertyValues {
         AlphaCurve {keys [<at 0 1 RCIM_Cubic>]}
         BlueCurve {keys [<at 0 1.7 RCIM_Cubic>,<at 0.6245 1>,<at 0.9795 0.3>]}
         GreenCurve {keys [<at 0 1.292 RCIM_Cubic>,<at 0.6245 0>,<at 0.9795 0.0047>]}
         RedCurve {keys [<at 0 0.17 RCIM_Cubic>,<at 0.6245 0.4907>,<at 0.9795 0.1031>]}
        }}))
     Day=&MakeVectorFromLinearColorRGB(Color=&ColorFromCurve(ColorCurve=DataInterface() {propertyValues {
         AlphaCurve {keys [<at 0 1 RCIM_Cubic>]}
         BlueCurve {keys [<at 0 3 RCIM_Cubic>,<at 0.6245 1>,<at 0.9795 0.3>]}
         GreenCurve {keys [<at 0 2.28 RCIM_Cubic>,<at 0.6245 0>,<at 0.9795 0.0047>]}
         RedCurve {keys [<at 0 0.3 RCIM_Cubic>,<at 0.6245 0.4907>,<at 0.9795 0.1031>]}
        }}))
     Evening=&MakeVectorFromLinearColorRGB(Color=&ColorFromCurve(ColorCurve=DataInterface() {propertyValues {
         AlphaCurve {keys [<at 0 1 RCIM_Cubic>]}
         BlueCurve {keys [<at 0 1.7 RCIM_Cubic>,<at 0.6245 1>,<at 0.9795 0.3>]}
         GreenCurve {keys [<at 0 1.292 RCIM_Cubic>,<at 0.6245 0>,<at 0.9795 0.0047>]}
         RedCurve {keys [<at 0 0.17 RCIM_Cubic>,<at 0.6245 0.4907>,<at 0.9795 0.1031>]}
        }}))
     Night=&MakeVectorFromLinearColorRGB(Color=&ColorFromCurve(ColorCurve=DataInterface() {propertyValues {
         AlphaCurve {keys [<at 0 1 RCIM_Cubic>]}
         BlueCurve {keys [<at 0 1.4 RCIM_Cubic>,<at 0.6245 1>,<at 0.9795 0.3>]}
         GreenCurve {keys [<at 0 1.273 RCIM_Cubic>,<at 0.6245 0>,<at 0.9795 0.0047>]}
         RedCurve {keys [<at 0 0.9293 RCIM_Cubic>,<at 0.6245 0.4907>,<at 0.9795 0.1031>]}
        }}))
    )
    ScaleA=true
    ScaleRGB=true
   ) {enabled=false, module=ScaleColor001}
   /SetVariables(Particles.RibbonWidth=&Multiply_Float(A=$Particles.RibbonWidth, B=&Lerp_Float(A=1, B=0.9, Alpha=$User.ParaGlideLeanAlpha))) {enabled=false, isSetParameters=true, module=SetVariables_#5}
   &ScaleRibbonWidth(Ribbon Width Scale=&FloatFromCurve(FloatCurve=DataInterface() {propertyValues {Curve {keys [
         {arriveTangent=-0.1351, interpMode=RCIM_Cubic, leaveTangent=-0.1351, tangentMode=RCTM_User, time=0.1, value=1}
         {arriveTangent=-0.6647, interpMode=RCIM_Cubic, leaveTangent=-0.6647, tangentMode=RCTM_User, time=1, value=0.6}
        ]}}}))
  ]
  rendererClasses [&NiagaraRibbonRendererProperties]
  renderers [{rendererClass=&NiagaraRibbonRendererProperties}]
 }
 Ribbons_Static {
  emitterSpawnScript [/SetVariables() {isSetParameters=true}]
  emitterUpdateScript [
   &EmitterState()
   &SpawnBurst_Instantaneous(Spawn Count=32, Spawn Group=1, Use Spawn Probability=false, UseLoopCountLimit=true)
   /SetVariables(Emitter.RibbonGrowth:NiagaraFloat=&FloatFromCurve(
     FloatCurve=DataInterface() {propertyValues {Curve {keys [<at 0 0 RCIM_Cubic>, {arriveTangent=2.7154, interpMode=RCIM_Cubic, leaveTangent=2.7154, tangentMode=RCTM_User, time=0.45, value=1}]}}}
     CurveIndex=$Emitter.Age
    )) {isSetParameters=true}
  ]
  particleSpawnScript [
   &PartitionParticles(Number Of Partitions=4)
   &InitializeParticle(
    Color Mode=::NewEnumerator0
    Ribbon Width=&MakeCustomFloatFromBool(
     Bool=&SetBoolByIntComparison(A=$Particles.PartitionParticles.Partition, B=1, Comparison Type=::NewEnumerator0)
     True Float=16
     False Float=34
    )
    Ribbon Width Mode=::NewEnumerator1
    Sprite Size Mode=::NewEnumerator3
    Uniform Sprite Size=20
    UsePositionOffset=false
    Write Material Random=true
   )
   /SetVariables(Particles.RibbonLinkOrder:NiagaraFloat=&ReturnNormalizedExecIndex()) {isSetParameters=true}
   &DynamicMaterialParameters(
    Index 0 Param 1=&Multiply_Float(A=&MakeFloatFromInt(Int=$Particles.PartitionParticles.Partition), B=0.2)
    Index 0 Param 2=&FN_FrontEndFloatModifier(Gameplay=1, Front End=0.8)
    Index 0 Param 3=&Multiply_Float(A=&MakeFloatFromInt(Int=$Particles.PartitionParticles.Partition), B=0.4)
    Index 0 Param 4=&MaskFloatBySpawnGroup(
     Masked Float=1
     Passthrough Float=0.5
     Spawn Group Mask=&ModuloInt(A=&SpawnGroup(), B=2)
    )
    Module.Param0WriteEnabled=true
    Module.Param1WriteEnabled=true
    Module.Param2WriteEnabled=true
    Module.Param3WriteEnabled=true
   )
  ]
  particleUpdateScript [
   &ParticleState(Kill Particles When Lifetime Has Elapsed=false, Let Infinitely Lived Particles Die When Emitter Deactivates=true)
   &SkeletalMeshLocation(
    Bone / Socket Index=$Particles.PartitionParticles.Partition
    Bone Sampling Mode=::NewEnumerator4
    MaskSpawnGroup=false
    Sampled Position Offset=&MakeCustomVectorFromBool(
     Bool=&SetBoolByIntComparison(A=$Particles.PartitionParticles.Partition, B=1, Comparison Type=::NewEnumerator0)
     True Vector=&Multiply_VectorByFloat(
      Vector=&FN_FrontEndVector3Modifier(
       Gameplay=&Multiply_VectorByFloat(Vector=$System.PiecesVelocity, Float=$System.PiecesOffsetDistance)
       Front End=&Multiply_VectorByFloat(Vector=&Lerp_Vector(A=<0 0 -1>, B=<0 1 -1>, Alpha=$User.ParaGlideLeanAlpha), Float=$System.PiecesOffsetDistance)
      )
      Float=&FloatFromCurve(
       FloatCurve=DataInterface() {propertyValues {Curve {keys [
           {arriveTangent=0.3395, interpMode=RCIM_Cubic, leaveTangent=0.3395, tangentMode=RCTM_User, time=0.2, value=0}
           {arriveTangent=2.7349, interpMode=RCIM_Cubic, leaveTangent=2.7349, tangentMode=RCTM_User, time=1, value=1}
          ]}}}
       CurveIndex=&ReturnNormalizedExecIndex()
       Scale Curve=&Multiply_Float(
        A=$System.PiecesOffsetDistance
        B=&Multiply_Float(A=&Lerp_Float(A=-0.5, B=-0.9, Alpha=$User.ParaGlideLeanAlpha), B=$Emitter.RibbonGrowth)
       )
      )
     )
     False Vector=&Multiply_VectorByFloat(
      Vector=&FN_FrontEndVector3Modifier(
       Gameplay=&Multiply_VectorByFloat(Vector=$System.PiecesVelocity, Float=$System.PiecesOffsetDistance)
       Front End=&Multiply_VectorByFloat(Vector=&Lerp_Vector(A=<0 0 -1>, B=<0 1 -1>, Alpha=$User.ParaGlideLeanAlpha), Float=$System.PiecesOffsetDistance)
      )
      Float=&FloatFromCurve(
       FloatCurve=DataInterface() {propertyValues {Curve {keys [
           {arriveTangent=0.1842, interpMode=RCIM_Cubic, leaveTangent=0.1842, tangentMode=RCTM_User, time=0.2, value=0}
           {arriveTangent=3.0363, interpMode=RCIM_Cubic, leaveTangent=3.0363, tangentMode=RCTM_User, time=1, value=1}
          ]}}}
       CurveIndex=&ReturnNormalizedExecIndex()
       Scale Curve=&Multiply_Float(
        A=$System.PiecesOffsetDistance
        B=&Multiply_Float(A=&Lerp_Float(A=-1.3, B=-2, Alpha=$User.ParaGlideLeanAlpha), B=$Emitter.RibbonGrowth)
       )
      )
     )
    )
    Sampled Position Offset Coordinate Space:ENiagaraSkelMeshTransforms=::NewEnumerator2
    Skeletal Mesh=$User.SkeletalMesh
   ) {module=SkeletalMeshLocation001}
   &ScaleColor(
    Scale Alpha=&FloatFromCurve(
     FloatCurve=DataInterface() {propertyValues {Curve {keys [
         {arriveTangent=-2.7341, interpMode=RCIM_Cubic, leaveTangent=-2.7341, tangentMode=RCTM_User, time=0.18, value=1}
         {arriveTangent=-4.7619, leaveTangent=-4.7619, time=0.35, value=-1}
         {arriveTangent=-0.3544, interpMode=RCIM_Cubic, leaveTangent=-0.3544, tangentMode=RCTM_User, time=0.6, value=-1}
         {arriveTangent=0.3316, interpMode=RCIM_Cubic, leaveTangent=0.3316, tangentMode=RCTM_User, time=1, value=1}
        ]}}}
     CurveIndex=&ReturnNormalizedExecIndex()
     Scale Curve=1.2
    )
    Scale RGB=&FN_Select_Vector_from_Time_Of_Day(
     Morning=&MakeVectorFromLinearColorRGB(Color=&ColorFromCurve(
       ColorCurve=DataInterface() {propertyValues {
         AlphaCurve {keys [<at 0 1 RCIM_Cubic>]}
         BlueCurve {keys [<at 0.3155 0.15>,<at 0.5806 0.7>,<at 0.9969 1.4742>]}
         GreenCurve {keys [<at 0.3155 0>,<at 0.5806 0>,<at 0.9969 0.03>]}
         RedCurve {keys [<at 0.3155 0.0321>,<at 0.5806 0.3188>,<at 0.9969 3>]}
        }}
       CurveIndex=&ReturnNormalizedExecIndex(Normalized Index Includes One=false)
      ))
     Day=&MakeVectorFromLinearColorRGB(Color=&ColorFromCurve(
       ColorCurve=DataInterface() {propertyValues {
         AlphaCurve {keys [<at 0 1 RCIM_Cubic>]}
         BlueCurve {keys [<at 0.3178 0.22>,<at 0.583 0.5>,<at 1 0.9828>]}
         GreenCurve {keys [<at 0.3178 0>,<at 0.583 0>,<at 1 0.02>]}
         RedCurve {keys [<at 0.3178 0.047>,<at 0.583 0.2277>,<at 1 2>]}
        }}
       CurveIndex=&ReturnNormalizedExecIndex(Normalized Index Includes One=false)
      ))
     Evening=&MakeVectorFromLinearColorRGB(Color=&ColorFromCurve(
       ColorCurve=DataInterface() {propertyValues {
         AlphaCurve {keys [<at 0 1 RCIM_Cubic>]}
         BlueCurve {keys [<at 0.3247 0.15>,<at 0.5806 0.7>,<at 0.9992 1.4742>]}
         GreenCurve {keys [<at 0.3247 0>,<at 0.5806 0>,<at 0.9992 0.03>]}
         RedCurve {keys [<at 0.3247 0.0321>,<at 0.5806 0.3188>,<at 0.9992 3>]}
        }}
       CurveIndex=&ReturnNormalizedExecIndex(Normalized Index Includes One=false)
      ))
     Night=&MakeVectorFromLinearColorRGB(Color=&ColorFromCurve(
       ColorCurve=DataInterface() {propertyValues {
         AlphaCurve {keys [<at 0 1 RCIM_Cubic>]}
         BlueCurve {keys [<at 0.3155 0.1>,<at 0.5806 0.45>,<at 1 0.4914>]}
         GreenCurve {keys [<at 0.3155 0>,<at 0.5806 0>,<at 1 0.01>]}
         RedCurve {keys [<at 0.3155 0.0214>,<at 0.5806 0.2049>,<at 1 1>]}
        }}
       CurveIndex=&ReturnNormalizedExecIndex(Normalized Index Includes One=false)
      ))
    )
    ScaleA=true
    ScaleRGB=true
   )
   &CameraOffset(Camera Offset Amount=10)
  ]
  rendererClasses [&NiagaraRibbonRendererProperties]
  renderers [{rendererClass=&NiagaraRibbonRendererProperties}]
 }
 Sparks {
  emitterUpdateScript [
   &EmitterState()
   &SpawnRate(SpawnRate=&Lerp_Float(A=12, B=40, Alpha=$System.Dive), Use Spawn Probability=false)
  ]
  particleSpawnScript [
   &InitializeParticle(
    Lifetime=&RandomRangeFloat(Minimum=0.35, Maximum=0.65)
    Sprite Size=&Multiply_Vector2D_ByFloat(Float=&RandomRangeFloat(Minimum=0.6, Maximum=1.3), Vector2D=<1 6>)
    Sprite Size Mode=::NewEnumerator4
    UsePositionOffset=false
    Write Material Random=true
   )
   &FN_SetSpawnGroup()
   &SkeletalMeshLocation(
    Bone / Socket Index=$Particles.SpawnGroup
    Bone Sampling Mode=::NewEnumerator4
    Inherit Velocity=false
    Interpolate Mesh Sampling=false
    MaskSpawnGroup=false
    Sampled Position Offset=&Multiply_Vector(
     A=&Lerp_Vector(A=<0 11 12>, B=<0 -7 12>, Alpha=$System.Dive)
     B=&Lerp_Vector(A=<1 1 1>, B=<1 1 3>, Alpha=$Particles.NormalizedAge)
    )
    Skeletal Mesh=$User.SkeletalMesh
   )
   &SphereLocation(
    MaskSpawnGroup=false
    Override Seed=false
    Sphere Origin:NiagaraPosition=$Output.SkeletalMeshLocation.SampledPosition
    Sphere Radius=3
    Surface Only=false
    UseNonUniformScale=false
    UseOffset=false
    UseSphereOrientation=false
   )
   &AddVelocity(Velocity=&FN_FrontEndVector3Modifier(
     Gameplay=&Multiply_VectorByFloat(
      Vector=&Lerp_Vector(
       A=&RandomRangeVector(Minimum=<0 0 300>, Maximum=<0 0 550>)
       B=&RandomRangeVector(Minimum=<0 0 600>, Maximum=<0 0 800>)
       Alpha=$System.Dive
      )
      Float=&RandomRangeFloat(Minimum=0.3, Maximum=0.8)
     )
     Front End=&Multiply_VectorByFloat(
      Vector=&Lerp_Vector(
       A=&RandomRangeVector(Minimum=<0 0 200>, Maximum=<0 0 500>)
       B=&RandomRangeVector(Minimum=<0 -300 300>, Maximum=<0 -600 700>)
       Alpha=$System.Dive
      )
      Float=&RandomRangeFloat(Minimum=0.3, Maximum=0.8)
     )
    )) {module=AddVelocity002}
  ]
  particleUpdateScript [
   &ParticleState()
   &InheritSourceMovement(
    Applied Force Scale=<-0.3 -0.3 0>
    ApplyForce=true
    ApplyForceFalloff=false
    ApplyPosition=true
    ApplyPositionFalloff=false
    Module.LimitForceSpeed=false
   ) {module=InheritSourceMovement001}
   /Niagara/Modules/Update/Forces/CurlNoiseForce(
    Noise Frequency=30
    Noise Quality / Cost=::NewEnumerator0
    Noise Strength=&FloatFromCurve(
     FloatCurve=DataInterface() {propertyValues {Curve {keys [{interpMode=RCIM_Cubic, tangentMode=RCTM_User, time=0.4, value=1}, {interpMode=RCIM_Cubic, tangentMode=RCTM_User, time=1, value=-1}]}}}
     Scale Curve=500
    )
    Pan Noise=false
   )
   &Drag(Drag=&RandomRangeFloat(Minimum=0.5, Maximum=1), Use Linear Drag=true, Use Rotational Drag=false) {enabled=false}
   &RotateAroundPoint(
    Radius:NiagaraFloat=&FloatFromCurve(
     FloatCurve=DataInterface() {propertyValues {Curve {keys [{interpMode=RCIM_Cubic, tangentMode=RCTM_User, time=0, value=0.25}, {interpMode=RCIM_Cubic, tangentMode=RCTM_User, time=1, value=1}]}}}
     Scale Curve=10
    )
    Rotation Rate:NiagaraFloat=&RandomRangeFloat(Minimum=0.5, Maximum=1)
    Set Particle Position=false
   ) {enabled=false}
   &SolveForcesAndVelocity(Clamp Velocity=false, Limit Acceleration=false)
   /SetVariables(Particles.Rotational_Position:NiagaraPosition=&Add_Vector(A=Linked() {linkedVariable {name=Particles.Position, type=NiagaraPosition}}, B=$Output.RotateAroundPoint.OutputPosition)) {enabled=false, isSetParameters=true, module=SetVariables_#6}
   &ScaleColor(
    Scale Alpha=&FloatFromCurve(FloatCurve=DataInterface() {propertyValues {Curve {keys [<at 0.7 1 RCIM_Cubic>,<at 1 0>]}}})
    Scale RGB=&MakeVectorFromLinearColorRGB(Color=&FN_Select_Color_from_Time_Of_Day(
      Morning=&ColorFromCurve(ColorCurve=DataInterface() {propertyValues {
         AlphaCurve {keys []}
         BlueCurve {keys [
           <at 0.0268 5>
           <at 0.1024 4>
           {arriveTangent=-4.4564, interpMode=RCIM_Cubic, leaveTangent=-4.4564, time=0.1652, value=3}
           <at 1 0 RCIM_Cubic>
          ]}
         GreenCurve {keys [
           <at 0.0268 3.3189>
           <at 0.1024 0.12>
           {arriveTangent=-0.1337, interpMode=RCIM_Cubic, leaveTangent=-0.1337, time=0.1652, value=0.059}
           <at 1 0 RCIM_Cubic>
          ]}
         RedCurve {keys [
           <at 0.0268 0.3479>
           <at 0.1024 0.12>
           {arriveTangent=-0.1337, interpMode=RCIM_Cubic, leaveTangent=-0.1337, time=0.1652, value=0.8053}
           <at 1 0 RCIM_Cubic>
          ]}
        }})
      Day=&ColorFromCurve(ColorCurve=DataInterface() {propertyValues {
         AlphaCurve {keys []}
         BlueCurve {keys [
           <at 0.0244 5>
           <at 0.1049 4>
           {arriveTangent=-4.4685, interpMode=RCIM_Cubic, leaveTangent=-4.4685, time=0.1652, value=4}
           <at 1 0 RCIM_Cubic>
          ]}
         GreenCurve {keys [
           <at 0.0244 3.3189>
           <at 0.1049 0.12>
           {arriveTangent=-0.1341, interpMode=RCIM_Cubic, leaveTangent=-0.1341, time=0.1652, value=0.0787}
           <at 1 0 RCIM_Cubic>
          ]}
         RedCurve {keys [
           <at 0.0244 0.3479>
           <at 0.1049 0.12>
           {arriveTangent=-0.1341, interpMode=RCIM_Cubic, leaveTangent=-0.1341, time=0.1652, value=1.0737}
           <at 1 0 RCIM_Cubic>
          ]}
        }})
      Evening=&ColorFromCurve(ColorCurve=DataInterface() {propertyValues {
         AlphaCurve {keys []}
         BlueCurve {keys [
           <at 0.0317 4>
           <at 0.0927 3>
           {arriveTangent=-3.3064, interpMode=RCIM_Cubic, leaveTangent=-3.3064, time=0.1652, value=2}
           <at 1 0 RCIM_Cubic>
          ]}
         GreenCurve {keys [
           <at 0.0317 2.6551>
           <at 0.0927 0.09>
           {arriveTangent=-0.0992, interpMode=RCIM_Cubic, leaveTangent=-0.0992, time=0.1652, value=0.0393}
           <at 1 0 RCIM_Cubic>
          ]}
         RedCurve {keys [
           <at 0.0317 0.2783>
           <at 0.0927 0.09>
           {arriveTangent=-0.0992, interpMode=RCIM_Cubic, leaveTangent=-0.0992, time=0.1652, value=0.5369}
           <at 1 0 RCIM_Cubic>
          ]}
        }})
      Night=&ColorFromCurve(ColorCurve=DataInterface() {propertyValues {
         AlphaCurve {keys []}
         BlueCurve {keys [
           <at 0.0293 4>
           <at 0.0878 3>
           {arriveTangent=-3.2887, interpMode=RCIM_Cubic, leaveTangent=-3.2887, time=0.1652, value=2}
           <at 1 0 RCIM_Cubic>
          ]}
         GreenCurve {keys [
           <at 0.0293 2.6551>
           <at 0.0878 0.09>
           {arriveTangent=-0.0987, interpMode=RCIM_Cubic, leaveTangent=-0.0987, time=0.1652, value=0.0393}
           <at 1 0 RCIM_Cubic>
          ]}
         RedCurve {keys [
           <at 0.0293 0.2783>
           <at 0.0878 0.09>
           {arriveTangent=-0.0987, interpMode=RCIM_Cubic, leaveTangent=-0.0987, time=0.1652, value=0.5369}
           <at 1 0 RCIM_Cubic>
          ]}
        }})
     ))
    ScaleA=true
    ScaleRGB=true
   )
  ]
  rendererClasses [&NiagaraSpriteRendererProperties]
  renderers [{rendererClass=&NiagaraSpriteRendererProperties}]
 }
 Sparks001 {
  emitterUpdateScript [
   &EmitterState()
   &SpawnRate(SpawnRate=&Lerp_Float(A=10, B=20, Alpha=$System.Dive), Use Spawn Probability=false)
  ]
  particleSpawnScript [
   &InitializeParticle(
    Color=&RandomRangeLinearColor(Minimum=<rgba 0 0.5921 2 1>, Maximum=<rgba 0.9091 0 2 1>)
    Lifetime=&RandomRangeFloat(Minimum=0.35, Maximum=0.65)
    Sprite Size Mode=::NewEnumerator1
    Uniform Sprite Size Max=32
    Uniform Sprite Size Min=16
    UsePositionOffset=false
    Write Material Random=true
   )
   &FN_SetSpawnGroup()
   &SkeletalMeshLocation(
    Bone / Socket Index=$Particles.SpawnGroup
    Bone Sampling Mode=::NewEnumerator4
    Inherit Velocity=false
    Interpolate Mesh Sampling=false
    MaskSpawnGroup=false
    Sampled Position Offset=&Multiply_Vector(
     A=&Lerp_Vector(A=<0 11 12>, B=<0 -7 12>, Alpha=$User.ParaGlideLeanAlpha)
     B=&Lerp_Vector(A=<1 1 1>, B=<1 1 3>, Alpha=$Particles.NormalizedAge)
    )
    Skeletal Mesh=$User.SkeletalMesh
   )
   &SphereLocation(
    MaskSpawnGroup=false
    Offset=<0 0 20>
    Override Seed=false
    Sphere Origin:NiagaraPosition=$Output.SkeletalMeshLocation.SampledPosition
    Sphere Radius=3
    Surface Only=false
    UseNonUniformScale=false
    UseOffset=true
    UseSphereOrientation=false
   )
   &AddVelocity(Velocity=&FN_FrontEndVector3Modifier(
     Gameplay=&Multiply_VectorByFloat(
      Vector=&Lerp_Vector(
       A=&RandomRangeVector(Minimum=<0 0 300>, Maximum=<0 0 550>)
       B=&RandomRangeVector(Minimum=<0 0 600>, Maximum=<0 0 800>)
       Alpha=$System.Dive
      )
      Float=&RandomRangeFloat(Minimum=0.1, Maximum=0.4)
     )
     Front End=&Multiply_VectorByFloat(
      Vector=&Lerp_Vector(
       A=&RandomRangeVector(Minimum=<0 0 200>, Maximum=<0 0 500>)
       B=&RandomRangeVector(Minimum=<0 -300 300>, Maximum=<0 -600 700>)
       Alpha=$System.Dive
      )
      Float=&RandomRangeFloat(Minimum=0.1, Maximum=0.4)
     )
    )) {module=AddVelocity002}
  ]
  particleUpdateScript [
   &ParticleState()
   &InheritSourceMovement(
    Applied Force Scale=<-0.3 -0.3 0>
    ApplyForce=true
    ApplyForceFalloff=false
    ApplyPosition=true
    ApplyPositionFalloff=false
    Module.LimitForceSpeed=false
   ) {module=InheritSourceMovement001}
   &CurlNoiseForce(
    Noise Frequency=30
    Noise Quality / Cost=::NewEnumerator0
    Noise Strength=&FloatFromCurve(
     FloatCurve=DataInterface() {propertyValues {Curve {keys [{interpMode=RCIM_Cubic, tangentMode=RCTM_User, time=0.4, value=1}, {interpMode=RCIM_Cubic, tangentMode=RCTM_User, time=1, value=-1}]}}}
     Scale Curve=500
    )
    Pan Noise=false
   ) {enabled=false}
   &Drag(Drag=&RandomRangeFloat(Minimum=0.5, Maximum=1), Use Linear Drag=true, Use Rotational Drag=false) {enabled=false}
   &RotateAroundPoint(
    Radius:NiagaraFloat=&FloatFromCurve(
     FloatCurve=DataInterface() {propertyValues {Curve {keys [{interpMode=RCIM_Cubic, tangentMode=RCTM_User, time=0, value=0.25}, {interpMode=RCIM_Cubic, tangentMode=RCTM_User, time=1, value=1}]}}}
     Scale Curve=10
    )
    Rotation Rate:NiagaraFloat=&RandomRangeFloat(Minimum=0.5, Maximum=1)
    Set Particle Position=false
   ) {enabled=false}
   &SolveForcesAndVelocity(Clamp Velocity=false, Limit Acceleration=false)
   /SetVariables(Particles.Rotational_Position:NiagaraPosition=&Add_Vector(A=Linked() {linkedVariable {name=Particles.Position, type=NiagaraPosition}}, B=$Output.RotateAroundPoint.OutputPosition)) {enabled=false, isSetParameters=true, module=SetVariables_#6}
   &ScaleColor(
    Scale Alpha=&FloatFromCurve(FloatCurve=DataInterface() {propertyValues {Curve {keys [<at 0.5 1 RCIM_Cubic>,<at 1 0>]}}})
    Scale RGB=&MakeVectorFromLinearColorRGB(Color=&FN_Select_Color_from_Time_Of_Day(
      Morning=&ColorFromCurve(ColorCurve=DataInterface() {propertyValues {
         AlphaCurve {keys []}
         BlueCurve {keys [
           <at 0.0268 5>
           <at 0.1024 4>
           {arriveTangent=-4.4564, interpMode=RCIM_Cubic, leaveTangent=-4.4564, time=0.1652, value=3}
           <at 1 0 RCIM_Cubic>
          ]}
         GreenCurve {keys [
           <at 0.0268 3.3189>
           <at 0.1024 0.12>
           {arriveTangent=-0.1337, interpMode=RCIM_Cubic, leaveTangent=-0.1337, time=0.1652, value=0.059}
           <at 1 0 RCIM_Cubic>
          ]}
         RedCurve {keys [
           <at 0.0268 0.3479>
           <at 0.1024 0.12>
           {arriveTangent=-0.1337, interpMode=RCIM_Cubic, leaveTangent=-0.1337, time=0.1652, value=0.8053}
           <at 1 0 RCIM_Cubic>
          ]}
        }})
      Day=&ColorFromCurve(ColorCurve=DataInterface() {propertyValues {
         AlphaCurve {keys []}
         BlueCurve {keys [
           <at 0.0244 5>
           <at 0.1049 4>
           {arriveTangent=-4.4685, interpMode=RCIM_Cubic, leaveTangent=-4.4685, time=0.1652, value=4}
           <at 1 0 RCIM_Cubic>
          ]}
         GreenCurve {keys [
           <at 0.0244 3.3189>
           <at 0.1049 0.12>
           {arriveTangent=-0.1341, interpMode=RCIM_Cubic, leaveTangent=-0.1341, time=0.1652, value=0.0787}
           <at 1 0 RCIM_Cubic>
          ]}
         RedCurve {keys [
           <at 0.0244 0.3479>
           <at 0.1049 0.12>
           {arriveTangent=-0.1341, interpMode=RCIM_Cubic, leaveTangent=-0.1341, time=0.1652, value=1.0737}
           <at 1 0 RCIM_Cubic>
          ]}
        }})
      Evening=&ColorFromCurve(ColorCurve=DataInterface() {propertyValues {
         AlphaCurve {keys []}
         BlueCurve {keys [
           <at 0.0317 4>
           <at 0.0927 3>
           {arriveTangent=-3.3064, interpMode=RCIM_Cubic, leaveTangent=-3.3064, time=0.1652, value=2}
           <at 1 0 RCIM_Cubic>
          ]}
         GreenCurve {keys [
           <at 0.0317 2.6551>
           <at 0.0927 0.09>
           {arriveTangent=-0.0992, interpMode=RCIM_Cubic, leaveTangent=-0.0992, time=0.1652, value=0.0393}
           <at 1 0 RCIM_Cubic>
          ]}
         RedCurve {keys [
           <at 0.0317 0.2783>
           <at 0.0927 0.09>
           {arriveTangent=-0.0992, interpMode=RCIM_Cubic, leaveTangent=-0.0992, time=0.1652, value=0.5369}
           <at 1 0 RCIM_Cubic>
          ]}
        }})
      Night=&ColorFromCurve(ColorCurve=DataInterface() {propertyValues {
         AlphaCurve {keys []}
         BlueCurve {keys [
           <at 0.0293 4>
           <at 0.0878 3>
           {arriveTangent=-3.2887, interpMode=RCIM_Cubic, leaveTangent=-3.2887, time=0.1652, value=2}
           <at 1 0 RCIM_Cubic>
          ]}
         GreenCurve {keys [
           <at 0.0293 2.6551>
           <at 0.0878 0.09>
           {arriveTangent=-0.0987, interpMode=RCIM_Cubic, leaveTangent=-0.0987, time=0.1652, value=0.0393}
           <at 1 0 RCIM_Cubic>
          ]}
         RedCurve {keys [
           <at 0.0293 0.2783>
           <at 0.0878 0.09>
           {arriveTangent=-0.0987, interpMode=RCIM_Cubic, leaveTangent=-0.0987, time=0.1652, value=0.5369}
           <at 1 0 RCIM_Cubic>
          ]}
        }})
     ))
    ScaleA=true
    ScaleRGB=false
   )
  ]
  rendererClasses [&NiagaraSpriteRendererProperties]
  renderers [{rendererClass=&NiagaraSpriteRendererProperties}]
 }
}
summary {
 systemName=NS_Contrail_SunLit_V3
 emitters [
  {emitterName=Flares, rendererClasses [&NiagaraSpriteRendererProperties]}
  {emitterName=Metallic_Pieces, rendererClasses [&NiagaraMeshRendererProperties]}
  {bEnabled=false, emitterName=FN_Debug001, rendererClasses [&NiagaraSpriteRendererProperties]}
  {emitterName=Feet_Pieces, rendererClasses [&NiagaraMeshRendererProperties]}
  {emitterName=Bands, rendererClasses [&NiagaraMeshRendererProperties, &NiagaraLightRendererProperties]}
  {emitterName=Lights, rendererClasses [&NiagaraSpriteRendererProperties]}
  {emitterName=Sparks, rendererClasses [&NiagaraSpriteRendererProperties]}
  {emitterName=Sparks001, rendererClasses [&NiagaraSpriteRendererProperties]}
  {bEnabled=false, emitterName=Ribbons001, rendererClasses [&NiagaraRibbonRendererProperties]}
  {bEnabled=false, emitterName=Ribbons002, rendererClasses [&NiagaraRibbonRendererProperties]}
  {emitterName=Ribbons_Static, rendererClasses [&NiagaraRibbonRendererProperties]}
 ]
 userVariables [
  User.bIsFrontEnd {type=NiagaraBool}
  User.bIsFrontEndPreview {type=NiagaraBool, defaultValue=true}
  User.ParaGlideLeanAlpha {type=NiagaraFloat, defaultValue=0, description=""}
  User.PlayerSpeed {type=NiagaraFloat, defaultValue=0, description=""}
  User.RibbonIDOffset {type=NiagaraInt32}
  User.SaplingAge {type=NiagaraInt32}
  User.SkeletalMesh {
   type=NiagaraDataInterfaceSkeletalMesh
   defaultValue {
    struct=NiagaraExt_VariableValue_DataInterface
    value {dataInterface=&self:NiagaraDataInterfaceSkeletalMesh_0}
   }
  }
  User.SkeletalMesh_Flow {
   type=NiagaraDataInterfaceSkeletalMesh
   defaultValue {
    struct=NiagaraExt_VariableValue_DataInterface
    value {dataInterface=&self:NiagaraDataInterfaceSkeletalMesh_2}
   }
  }
 ]
}
